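# Animation Blueprint for the CoconutShell glider, a creature that flaps, blinks and caws while the player skydives.
# Locomotion runs Closed, Open_Into and Idle_Open from parachute state or force-open, with cosmetic-preview variants.
# Curve-weighted additives layer flaps, leans, aim turns, head darts, blinks, caws, jitter and wing sway before a FrontEnd-only Glider_Base rescale.
# An unconnected Locomotion_GliderItem machine still references LocalZillaNight animations.
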AnimBlueprint /BRCosmetics/Animation/Game/MainPlayer/Skydive/ParaGlide/CoconutShell/CoconutShell_Glider_AnimBP
target_skeleton=/BRCosmetics/Gadgets/Assets/VinderTech_GliderChute/Glider_CoconutShell/Meshes/CoconutShell_Glider_Skeleton
graphs [
 AnimGraph {nodes [
   n1 = Break Fort Anim Input Skydiving!(FortAnimInput_Skydiving:"Fort Anim Input Skydiving Structure (by ref)"=Get SkydivingState() -> (SkydivingState:Fort Anim Input Skydiving Structure)) -> (
    ParaGilderRootOffset:Vector=~
    ParaGliderRootRotationOverride:Rotator=~
    bUseParaGlideRootModifier:bool=~
    bIsSkydivingFromLaunchPad:bool=~
    bIsSkydivingFromBus:bool=~
    bIsInVortex:bool=~
    bIsDiving:bool=~
    bIsParachuteOpen:bool=~
    bIsSkydiving:bool=~
    bIsParachuteLeaning:bool=~
    bIsSkydiveLeaning:bool=~
    bIsLeaning:bool=~
    bIsSkydiveDiveMode:bool=~
    bPlaySkydiveDrift:bool=~
    bSkydiveDriftDelayActive:bool=~
    bSkydiveDriftAnimAllowed:bool=~
    bIsGliderRight:bool=~
    bIsGliderCenter:bool=~
    bIsGliderLeft:bool=~
    bIsGliderForward:bool=~
    bIsGliderBack:bool=~
    bHasDeployedAndIsStandingUpright:bool=~
    LocalAccelForward:float=~
    LocalAccelRight:float=~
    LocalVelocityRight:float=~
    SkydiveAimPitch:float=~
    SkydiveAimYaw:float=~
    DeployChuteAnimRate:float=~
    SkydiveDriftAnimRate:float=~
    SkydiveDriftAnimRateCurrent:float=~
    SkydiveFidgetAnimRate:float=~
    SkydiveFidgetAnimRateCurrent:float=~
    SkydiveAdditiveAlpha:float=~
    SkydiveDriftDelay:float=~
    SkydiveDriftAnim:int
    SkydiveDriftAnimMax:int
    LaunchpadAnim:int
    LocalAccelDir:ESkydivingDirection Enum
    DirectionLast:ESkydivingDirection Enum
    PoseAdditive:Anim Sequence Object Reference
   ) {is=title K2Node_BreakStruct}
   Blendspace Player 'BS_CoconutShell_Lean_GLIDER'(X=@n1.LocalAccelRight, Y=@n1.LocalAccelForward) {blendSpace=&/BS_CoconutShell_Lean_GLIDER}
   AnimGraphNode_StateMachine_1 = Locomotion() {is=title AnimGraphNode_StateMachine}
   AnimGraphNode_StateMachine_6 = Flaps() {is=title AnimGraphNode_StateMachine}
   n2 = Apply Additive!(Base=@AnimGraphNode_StateMachine_1, Additive=@AnimGraphNode_StateMachine_6, AlphaCurveName:Name=CoconutShell_additive) -> (Pose) {alphaInputType=Curve, alphaBoolBlend {blendOption=Linear}, alphaScaleBiasClamp {bias=1, scale=-1}}
   AnimGraphNode_StateMachine_8 = LeanTransitions() {is=title AnimGraphNode_StateMachine}
   n3 = Break Fort Anim Input Skydiving(FortAnimInput_Skydiving=Get SkydivingState()) -> (
    ParaGilderRootOffset=~
    ParaGliderRootRotationOverride=~
    bUseParaGlideRootModifier=~
    bIsSkydivingFromLaunchPad=~
    bIsSkydivingFromBus=~
    bIsInVortex=~
    bIsDiving=~
    bIsParachuteOpen=~
    bIsSkydiving=~
    bIsParachuteLeaning=~
    bIsSkydiveLeaning=~
    bIsLeaning=~
    bIsSkydiveDiveMode=~
    bPlaySkydiveDrift=~
    bSkydiveDriftDelayActive=~
    bSkydiveDriftAnimAllowed=~
    bIsGliderRight=~
    bIsGliderCenter=~
    bIsGliderLeft=~
    bIsGliderForward=~
    bIsGliderBack=~
    bHasDeployedAndIsStandingUpright=~
    LocalAccelForward=~
    LocalAccelRight=~
    LocalVelocityRight=~
    SkydiveAimPitch=~
    SkydiveAimYaw=~
    DeployChuteAnimRate=~
    SkydiveDriftAnimRate=~
    SkydiveDriftAnimRateCurrent=~
    SkydiveFidgetAnimRate=~
    SkydiveFidgetAnimRateCurrent=~
    SkydiveAdditiveAlpha=~
    SkydiveDriftDelay=~
   )
   n4 = Blendspace Player 'BS_CoconutShell_Turn_GLIDER'(X=@n3.SkydiveAimYaw, Y=@n3.SkydiveAimPitch) {blendSpace=&/BS_CoconutShell_Turn_GLIDER}
   n5 = "Sequence Player (additive) 'CoconutShell_PositionAdjustement_Pose_GLIDER'"() {
    bLoopAnimation=false
    sequence=&/CoconutShell_PositionAdjustement_Pose_GLIDER
   }
   AnimGraphNode_StateMachine_3 = Head Darts() {is=title AnimGraphNode_StateMachine}
   n6 = Apply Additive(
    Base=Apply Additive(
     Base=Apply Additive(Base=@n2, Additive=@AnimGraphNode_StateMachine_8, AlphaCurveName=ParaGlide_Lean) {alphaInputType=Curve, alphaBoolBlend {blendOption=Linear}, alphaScaleBiasClamp {bias=1, scale=-1}}
     Additive=Apply Additive!(Base=@n4, Additive=@n5, Alpha:float) -> (Pose) {alphaBoolBlend {blendOption=Linear}, alphaScaleBiasClamp {scale=0.3}}
     AlphaCurveName=ParaGlide_Lean
    ) {alphaInputType=Curve, alphaBoolBlend {blendOption=Linear}, alphaScaleBiasClamp {bias=1, scale=-1}}
    Additive=@AnimGraphNode_StateMachine_3
    AlphaCurveName=CoconutShell_additive
   ) {alphaInputType=Curve, bAlphaBoolEnabled=false, alphaBoolBlend {blendOption=Linear}, alphaScaleBiasClamp {bias=1, scale=-1}}
   AnimGraphNode_StateMachine_4 = Blinks() {is=title AnimGraphNode_StateMachine}
   n7 = Apply Additive!(Base=@n6, Additive=@AnimGraphNode_StateMachine_4, AlphaCurveName:Name=CoconutShell_additive) -> (Pose) {alphaInputType=Curve, bAlphaBoolEnabled=false, alphaBoolBlend {blendOption=Linear}, alphaScaleBiasClamp {bias=1, scale=-1}}
   AnimGraphNode_StateMachine_5 = Caw() {is=title AnimGraphNode_StateMachine}
   n8 = Apply Additive(
    Base=Apply Additive(
     Base=Apply Additive(Base=@n7, Additive=@AnimGraphNode_StateMachine_5, AlphaCurveName=CoconutShell_additive) {alphaInputType=Curve, alphaBoolBlend {blendOption=Linear}, alphaScaleBiasClamp {bias=1, scale=-1}}
     Additive="Sequence Player (additive) 'CoconutShell_Jitter_Anim_GLIDER'"!(PlayRateBasis:float=1) -> (Pose) {playRateBasis=1.2, sequence=&/CoconutShell_Jitter_Anim_GLIDER}
     AlphaCurveName=additivejitter
    ) {alphaInputType=Curve, alphaBoolBlend {blendOption=Linear}, alphaScaleBiasClamp {bias=1, scale=-1}}
    Additive="Sequence Player (additive) 'CoconutShell_WingSway_Anim_GLIDER'"!(PlayRateBasis:float=1) -> (Pose) {sequence=&/CoconutShell_WingSway_Anim_GLIDER}
    AlphaCurveName=additivejitter
   ) {alphaInputType=Curve, alphaBoolBlend {blendOption=Linear}, alphaScaleBiasClamp {bias=1, scale=-1}}
   n9 = "Sequence Player (additive) 'CoconutShell_PositionAdjustement_Pose_GLIDER'"() {sequence=&CoconutShell_PositionAdjustement_Pose_GLIDER}
   n10 = "Transform (Modify) Bone - Bone: Glider_Base"!(
    ComponentPose=Local To Component(LocalPose=Apply Additive(Base=@n8, Additive=@n9, AlphaCurveName=ParaGlide_Lean) {alphaInputType=Curve, alphaBoolBlend {blendOption=Linear}, alphaScaleBiasClamp {bias=0.25, scale=-1}})
    bAlphaBoolEnabled:bool=Get FrontEnd() -> (FrontEnd:bool)
    Translation="0,0,0"
    Rotation="0,0,0"
    Scale="0.5,0.5,0.5"
   ) -> (Pose) {alphaInputType=Bool, scaleMode=BMM_Replace, alphaBoolBlend {blendOption=Linear}, boneToModify {boneName=Glider_Base}, scale=<0.5 0.5 0.5>}
   Output Pose(Result=Component To Local(ComponentPose=@n10)) {layerGroup=None}
   AnimGraphNode_StateMachine_2 = Locomotion_GliderItem() {is=title AnimGraphNode_StateMachine}
  ]}
 Locomotion {
  entry=Closed
  path=AnimGraph.AnimGraphNode_StateMachine_1.Locomotion
  ids=id
  nodes [
   AnimStateNode_0 = Idle_Open() {is=title AnimStateNode}
   AnimStateNode_2 = Closed() {is=title AnimStateNode, stateEntered {guid=#1, notifyName=ClosedEntered}}
   AnimStateNode_3 = Open_Into() {is=title AnimStateNode, stateEntered {guid=#2, notifyName=DeployEntered}}
   AnimStateTransitionNode_0 = AnimStateTransitionNode() {automaticRuleTriggerTime=-0.1, bAutomaticRuleBasedOnSequencePlayerInState=true, blendMode=Sinusoidal, crossfadeDuration=0.01}
   AnimStateTransitionNode_1 = AnimStateTransitionNode() {crossfadeDuration=0}
   AnimStateTransitionNode_2 = AnimStateTransitionNode() {crossfadeDuration=0}
   AnimStateTransitionNode_3 = AnimStateTransitionNode() {crossfadeDuration=0}
  ]
  links [
   {
    dst=Open_Into
    rule=AnimGraph.AnimGraphNode_StateMachine_1.Locomotion.AnimStateTransitionNode_2.Transition
    src=Closed
    via=AnimStateTransitionNode_2
   }
   {
    dst=Closed
    rule=AnimGraph.AnimGraphNode_StateMachine_1.Locomotion.AnimStateTransitionNode_1.Transition
    src=Idle_Open
    via=AnimStateTransitionNode_1
   }
   {
    dst=Closed
    rule=AnimGraph.AnimGraphNode_StateMachine_1.Locomotion.AnimStateTransitionNode_3.Transition
    src=Open_Into
    via=AnimStateTransitionNode_3
   }
   {
    dst=Idle_Open
    rule=AnimGraph.AnimGraphNode_StateMachine_1.Locomotion.AnimStateTransitionNode_0.Transition
    src=Open_Into
    via=AnimStateTransitionNode_0
   }
  ]
 }
 Idle_Open {
  path=AnimGraph.AnimGraphNode_StateMachine_1.Locomotion.AnimStateNode_0.Idle_Open
  nodes [
   Apply Additive!(
    Base
    Additive
    bAlphaBoolEnabled:bool=Get IsCosmeticPreview!(self:Fort Player Parachute Object Reference=Get OwnerGlider() -> (OwnerGlider:Fort Player Parachute Object Reference)) -> (IsCosmeticPreview:bool)
   ) -> (Pose) {alphaInputType=Bool, alphaBoolBlend {blendOption=Linear}}
   n1 = Break Fort Anim Input Skydiving(FortAnimInput_Skydiving=Get SkydivingState()) -> (
    ParaGilderRootOffset=~
    ParaGliderRootRotationOverride=~
    bUseParaGlideRootModifier=~
    bIsSkydivingFromLaunchPad=~
    bIsSkydivingFromBus=~
    bIsInVortex=~
    bIsDiving=~
    bIsParachuteOpen=~
    bIsSkydiving=~
    bIsParachuteLeaning=~
    bIsSkydiveLeaning=~
    bIsLeaning=~
    bIsSkydiveDiveMode=~
    bPlaySkydiveDrift=~
    bSkydiveDriftDelayActive=~
    bSkydiveDriftAnimAllowed=~
    bIsGliderRight=~
    bIsGliderCenter=~
    bIsGliderLeft=~
    bIsGliderForward=~
    bIsGliderBack=~
    bHasDeployedAndIsStandingUpright=~
    LocalAccelForward=~
    LocalAccelRight=~
    LocalVelocityRight=~
    SkydiveAimPitch=~
    SkydiveAimYaw=~
    DeployChuteAnimRate=~
    SkydiveDriftAnimRate=~
    SkydiveDriftAnimRateCurrent=~
    SkydiveFidgetAnimRate=~
    SkydiveFidgetAnimRateCurrent=~
    SkydiveAdditiveAlpha=~
    SkydiveDriftDelay=~
   )
   n2 = Get OwnerGlider()
   Output Animation Pose(Result=Blend Poses by bool(
     BlendPose_0=Sequence Player 'CoconutShell_Glide_FrontEnd_Anim_GLIDER'() {sequence=&/CoconutShell_Glide_FrontEnd_Anim_GLIDER}
     BlendPose_1=Blendspace Player 'BS_CoconutShell_GLIDER'!(X=@n1.LocalAccelRight, Y=@n1.LocalAccelForward, PlayRate:float=1) -> (Pose) {blendSpace=&/BS_CoconutShell_GLIDER}
     bActiveValue=Get IsCosmeticPreview(self=@n2)
    ) {blendType=Linear})
  ]
 }
 Closed {
  path=AnimGraph.AnimGraphNode_StateMachine_1.Locomotion.AnimStateNode_2.Closed
  nodes [
   n1 = Get OwnerGlider()
   Output Animation Pose(Result=Blend Poses by bool(
     BlendPose_0=Sequence Player 'CoconutShell_Closed_FrontEnd_Pose_GLIDER'() {sequence=&/CoconutShell_Closed_FrontEnd_Pose_GLIDER}
     BlendPose_1=Sequence Player 'CoconutShell_Closed_Pose_GLIDER'() {sequence=&/CoconutShell_Closed_Pose_GLIDER}
     bActiveValue=Get IsCosmeticPreview(self=@n1)
    ) {blendType=Linear})
  ]
 }
 Open_Into {
  path=AnimGraph.AnimGraphNode_StateMachine_1.Locomotion.AnimStateNode_3.Open_Into
  nodes [
   Apply Additive(bAlphaBoolEnabled=Get IsCosmeticPreview(self=Get OwnerGlider())) {alphaInputType=Bool, alphaBoolBlend {blendOption=Linear}}
   n1 = Break Fort Anim Input Skydiving(FortAnimInput_Skydiving=Get SkydivingState()) -> (
    ParaGilderRootOffset=~
    ParaGliderRootRotationOverride=~
    bUseParaGlideRootModifier=~
    bIsSkydivingFromLaunchPad=~
    bIsSkydivingFromBus=~
    bIsInVortex=~
    bIsDiving=~
    bIsParachuteOpen=~
    bIsSkydiving=~
    bIsParachuteLeaning=~
    bIsSkydiveLeaning=~
    bIsLeaning=~
    bIsSkydiveDiveMode=~
    bPlaySkydiveDrift=~
    bSkydiveDriftDelayActive=~
    bSkydiveDriftAnimAllowed=~
    bIsGliderRight=~
    bIsGliderCenter=~
    bIsGliderLeft=~
    bIsGliderForward=~
    bIsGliderBack=~
    bHasDeployedAndIsStandingUpright=~
    LocalAccelForward=~
    LocalAccelRight=~
    LocalVelocityRight=~
    SkydiveAimPitch=~
    SkydiveAimYaw=~
    DeployChuteAnimRate=~
    SkydiveDriftAnimRate=~
    SkydiveDriftAnimRateCurrent=~
    SkydiveFidgetAnimRate=~
    SkydiveFidgetAnimRateCurrent=~
    SkydiveAdditiveAlpha=~
    SkydiveDriftDelay=~
   )
   n2 = Sequence Player 'CoconutShell_Into_FrontEnd_Anim_GLIDER'() {bLoopAnimation=false, sequence=&/CoconutShell_Into_FrontEnd_Anim_GLIDER}
   n3 = Get OwnerGlider()
   Output Animation Pose(Result=Blend Poses by bool(
     BlendPose_0=@n2
     BlendPose_1=Blendspace Player 'BS_CoconutShell_Into_GLIDER'(X=@n1.SkydiveAimYaw, Y=@n1.SkydiveAimPitch) {blendSpace=&/BS_CoconutShell_Into_GLIDER}
     bActiveValue=Get IsCosmeticPreview(self=@n3)
    ) {blendType=Linear})
  ]
 }
 Transition {
  path=AnimGraph.AnimGraphNode_StateMachine_1.Locomotion.AnimStateTransitionNode_0.Transition
  nodes [
   Result()
   "Equal (Float)"!(A:double="Get Relevant Anim Time Remaining (Open_Into)"!() -> (ReturnValue:float) {is=title K2Node_AnimGetter}, B:double=0) -> (ReturnValue:bool) {is=title K2Node_CallFunction}
  ]
 }
 Transition {
  path=AnimGraph.AnimGraphNode_StateMachine_1.Locomotion.AnimStateTransitionNode_1.Transition
  nodes [
   n1 = Get bForceOpen() -> (bForceOpen:bool)
   n2 = Break Fort Anim Input Skydiving(FortAnimInput_Skydiving=Get SkydivingState()) -> (
    ParaGilderRootOffset=~
    ParaGliderRootRotationOverride=~
    bUseParaGlideRootModifier=~
    bIsSkydivingFromLaunchPad=~
    bIsSkydivingFromBus=~
    bIsInVortex=~
    bIsDiving=~
    bIsParachuteOpen=~
    bIsSkydiving=~
    bIsParachuteLeaning=~
    bIsSkydiveLeaning=~
    bIsLeaning=~
    bIsSkydiveDiveMode=~
    bPlaySkydiveDrift=~
    bSkydiveDriftDelayActive=~
    bSkydiveDriftAnimAllowed=~
    bIsGliderRight=~
    bIsGliderCenter=~
    bIsGliderLeft=~
    bIsGliderForward=~
    bIsGliderBack=~
    bHasDeployedAndIsStandingUpright=~
    LocalAccelForward=~
    LocalAccelRight=~
    LocalVelocityRight=~
    SkydiveAimPitch=~
    SkydiveAimYaw=~
    DeployChuteAnimRate=~
    SkydiveDriftAnimRate=~
    SkydiveDriftAnimRateCurrent=~
    SkydiveFidgetAnimRate=~
    SkydiveFidgetAnimRateCurrent=~
    SkydiveAdditiveAlpha=~
    SkydiveDriftDelay=~
   )
   Result(bCanEnterTransition=AND Boolean(A=NOT Boolean(A=@n2.bIsParachuteOpen), B=NOT Boolean(A=@n1)))
  ]
 }
 Transition {
  path=AnimGraph.AnimGraphNode_StateMachine_1.Locomotion.AnimStateTransitionNode_2.Transition
  nodes [
   n1 = Get SkydivingState()
   Result(bCanEnterTransition=OR Boolean(
     A=Break Fort Anim Input Skydiving(FortAnimInput_Skydiving=@n1) -> (
      ParaGilderRootOffset=~
      ParaGliderRootRotationOverride=~
      bUseParaGlideRootModifier=~
      bIsSkydivingFromLaunchPad=~
      bIsSkydivingFromBus=~
      bIsInVortex=~
      bIsDiving=~
      bIsParachuteOpen=~
      bIsSkydiving=~
      bIsParachuteLeaning=~
      bIsSkydiveLeaning=~
      bIsLeaning=~
      bIsSkydiveDiveMode=~
      bPlaySkydiveDrift=~
      bSkydiveDriftDelayActive=~
      bSkydiveDriftAnimAllowed=~
      bIsGliderRight=~
      bIsGliderCenter=~
      bIsGliderLeft=~
      bIsGliderForward=~
      bIsGliderBack=~
      bHasDeployedAndIsStandingUpright=~
      LocalAccelForward=~
      LocalAccelRight=~
      LocalVelocityRight=~
      SkydiveAimPitch=~
      SkydiveAimYaw=~
      DeployChuteAnimRate=~
      SkydiveDriftAnimRate=~
      SkydiveDriftAnimRateCurrent=~
      SkydiveFidgetAnimRate=~
      SkydiveFidgetAnimRateCurrent=~
      SkydiveAdditiveAlpha=~
      SkydiveDriftDelay=~
     ).bIsParachuteOpen
     B=Get bForceOpen()
    ))
  ]
 }
 Transition {
  path=AnimGraph.AnimGraphNode_StateMachine_1.Locomotion.AnimStateTransitionNode_3.Transition
  nodes [
   n1 = Get bForceOpen()
   n2 = Break Fort Anim Input Skydiving(FortAnimInput_Skydiving=Get SkydivingState()) -> (
    ParaGilderRootOffset=~
    ParaGliderRootRotationOverride=~
    bUseParaGlideRootModifier=~
    bIsSkydivingFromLaunchPad=~
    bIsSkydivingFromBus=~
    bIsInVortex=~
    bIsDiving=~
    bIsParachuteOpen=~
    bIsSkydiving=~
    bIsParachuteLeaning=~
    bIsSkydiveLeaning=~
    bIsLeaning=~
    bIsSkydiveDiveMode=~
    bPlaySkydiveDrift=~
    bSkydiveDriftDelayActive=~
    bSkydiveDriftAnimAllowed=~
    bIsGliderRight=~
    bIsGliderCenter=~
    bIsGliderLeft=~
    bIsGliderForward=~
    bIsGliderBack=~
    bHasDeployedAndIsStandingUpright=~
    LocalAccelForward=~
    LocalAccelRight=~
    LocalVelocityRight=~
    SkydiveAimPitch=~
    SkydiveAimYaw=~
    DeployChuteAnimRate=~
    SkydiveDriftAnimRate=~
    SkydiveDriftAnimRateCurrent=~
    SkydiveFidgetAnimRate=~
    SkydiveFidgetAnimRateCurrent=~
    SkydiveAdditiveAlpha=~
    SkydiveDriftDelay=~
   )
   Result(bCanEnterTransition=AND Boolean(A=NOT Boolean(A=@n2.bIsParachuteOpen), B=NOT Boolean(A=@n1)))
  ]
 }
 Locomotion_GliderItem {
  entry=Closed
  path=AnimGraph.AnimGraphNode_StateMachine_2.Locomotion_GliderItem
  ids=id
  nodes [
   AnimStateNode_0 = Idle_Open()
   AnimStateNode_1 = ForceOpen_or_GliderItem() {is=title AnimStateNode, stateEntered {guid=#3, notifyName=DeployEntered}}
   AnimStateNode_2 = Closed() {stateEntered {guid=#1, notifyName=ClosedEntered}}
   AnimStateNode_3 = Open_Into() {stateEntered {guid=#2, notifyName=DeployEntered}}
   AnimStateTransitionNode_0 = AnimStateTransitionNode() {bAutomaticRuleBasedOnSequencePlayerInState=true, blendMode=Linear, crossfadeDuration=0}
   AnimStateTransitionNode_1 = AnimStateTransitionNode() {crossfadeDuration=0}
   AnimStateTransitionNode_2 = AnimStateTransitionNode() {crossfadeDuration=0}
   AnimStateTransitionNode_3 = AnimStateTransitionNode() {crossfadeDuration=0, priorityOrder=2}
   AnimStateTransitionNode_4 = AnimStateTransitionNode() {crossfadeDuration=0}
   AnimStateTransitionNode_5 = AnimStateTransitionNode() {bAutomaticRuleBasedOnSequencePlayerInState=true, crossfadeDuration=0}
   AnimStateTransitionNode_6 = AnimStateTransitionNode() {crossfadeDuration=0}
  ]
  links [
   {
    dst=ForceOpen_or_GliderItem
    rule=AnimGraph.AnimGraphNode_StateMachine_2.Locomotion_GliderItem.AnimStateTransitionNode_4.Transition
    src=Closed
    via=AnimStateTransitionNode_4
   }
   {
    dst=Open_Into
    rule=AnimGraph.AnimGraphNode_StateMachine_2.Locomotion_GliderItem.AnimStateTransitionNode_1.Transition
    src=Closed
    via=AnimStateTransitionNode_1
   }
   {
    dst=Closed
    rule=AnimGraph.AnimGraphNode_StateMachine_2.Locomotion_GliderItem.AnimStateTransitionNode_6.Transition
    src=ForceOpen_or_GliderItem
    via=AnimStateTransitionNode_6
   }
   {
    dst=Idle_Open
    rule=AnimGraph.AnimGraphNode_StateMachine_2.Locomotion_GliderItem.AnimStateTransitionNode_5.Transition
    src=ForceOpen_or_GliderItem
    via=AnimStateTransitionNode_5
   }
   {
    dst=Closed
    rule=AnimGraph.AnimGraphNode_StateMachine_2.Locomotion_GliderItem.AnimStateTransitionNode_2.Transition
    src=Idle_Open
    via=AnimStateTransitionNode_2
   }
   {
    dst=Closed
    rule=AnimGraph.AnimGraphNode_StateMachine_2.Locomotion_GliderItem.AnimStateTransitionNode_3.Transition
    src=Open_Into
    via=AnimStateTransitionNode_3
   }
   {
    dst=Idle_Open
    rule=AnimGraph.AnimGraphNode_StateMachine_2.Locomotion_GliderItem.AnimStateTransitionNode_0.Transition
    src=Open_Into
    via=AnimStateTransitionNode_0
   }
  ]
 }
 Idle_Open {
  path=AnimGraph.AnimGraphNode_StateMachine_2.Locomotion_GliderItem.AnimStateNode_0.Idle_Open
  nodes [
   Apply Additive(bAlphaBoolEnabled=Get IsCosmeticPreview(self=Get OwnerGlider())) {alphaInputType=Bool, alphaBoolBlend {blendOption=Linear}}
   n1 = Break Fort Anim Input Skydiving(FortAnimInput_Skydiving=Get SkydivingState()) -> (
    ParaGilderRootOffset=~
    ParaGliderRootRotationOverride=~
    bUseParaGlideRootModifier=~
    bIsSkydivingFromLaunchPad=~
    bIsSkydivingFromBus=~
    bIsInVortex=~
    bIsDiving=~
    bIsParachuteOpen=~
    bIsSkydiving=~
    bIsParachuteLeaning=~
    bIsSkydiveLeaning=~
    bIsLeaning=~
    bIsSkydiveDiveMode=~
    bPlaySkydiveDrift=~
    bSkydiveDriftDelayActive=~
    bSkydiveDriftAnimAllowed=~
    bIsGliderRight=~
    bIsGliderCenter=~
    bIsGliderLeft=~
    bIsGliderForward=~
    bIsGliderBack=~
    bHasDeployedAndIsStandingUpright=~
    LocalAccelForward=~
    LocalAccelRight=~
    LocalVelocityRight=~
    SkydiveAimPitch=~
    SkydiveAimYaw=~
    DeployChuteAnimRate=~
    SkydiveDriftAnimRate=~
    SkydiveDriftAnimRateCurrent=~
    SkydiveFidgetAnimRate=~
    SkydiveFidgetAnimRateCurrent=~
    SkydiveAdditiveAlpha=~
    SkydiveDriftDelay=~
   )
   n2 = Get OwnerGlider()
   Output Animation Pose(Result=Blend Poses by bool(
     BlendPose_0=Sequence Player 'LocalZillaNight_SurfStyle_Glide_FrontEnd_GLIDER'() {sequence=/BRCosmetics/Animation/Game/MainPlayer/Skydive/ParaGlide/LocalZillaNight/LocalZillaNight_SurfStyle_Glide_FrontEnd_GLIDER}
     BlendPose_1=Blendspace Player 'BS_LocalZillaNight_SurfStyle_GLIDER'(X=@n1.SkydiveAimYaw, Y=@n1.SkydiveAimPitch) {blendSpace=/BRCosmetics/Animation/Game/MainPlayer/Skydive/ParaGlide/LocalZillaNight/BS_LocalZillaNight_SurfStyle_GLIDER}
     bActiveValue=Get IsCosmeticPreview(self=@n2)
    ) {blendType=Linear})
  ]
 }
 ForceOpen_or_GliderItem {
  path=AnimGraph.AnimGraphNode_StateMachine_2.Locomotion_GliderItem.AnimStateNode_1.ForceOpen_or_GliderItem
  nodes [
   n1 = Break Fort Anim Input Skydiving(FortAnimInput_Skydiving=Get SkydivingState()) -> (
    ParaGilderRootOffset=~
    ParaGliderRootRotationOverride=~
    bUseParaGlideRootModifier=~
    bIsSkydivingFromLaunchPad=~
    bIsSkydivingFromBus=~
    bIsInVortex=~
    bIsDiving=~
    bIsParachuteOpen=~
    bIsSkydiving=~
    bIsParachuteLeaning=~
    bIsSkydiveLeaning=~
    bIsLeaning=~
    bIsSkydiveDiveMode=~
    bPlaySkydiveDrift=~
    bSkydiveDriftDelayActive=~
    bSkydiveDriftAnimAllowed=~
    bIsGliderRight=~
    bIsGliderCenter=~
    bIsGliderLeft=~
    bIsGliderForward=~
    bIsGliderBack=~
    bHasDeployedAndIsStandingUpright=~
    LocalAccelForward=~
    LocalAccelRight=~
    LocalVelocityRight=~
    SkydiveAimPitch=~
    SkydiveAimYaw=~
    DeployChuteAnimRate=~
    SkydiveDriftAnimRate=~
    SkydiveDriftAnimRateCurrent=~
    SkydiveFidgetAnimRate=~
    SkydiveFidgetAnimRateCurrent=~
    SkydiveAdditiveAlpha=~
    SkydiveDriftDelay=~
   )
   n2 = Blendspace Player 'BS_LocalZillaNight_SurfStyle_Into_GLIDER'(X=@n1.SkydiveAimYaw, Y=@n1.SkydiveAimPitch) {
    bLoop=false
    blendSpace=/BRCosmetics/Animation/Game/MainPlayer/Skydive/ParaGlide/LocalZillaNight/BS_LocalZillaNight_SurfStyle_Into_GLIDER
    startPosition=0.0063
   }
   Output Animation Pose(Result=@n2)
  ]
 }
 Closed {
  path=AnimGraph.AnimGraphNode_StateMachine_2.Locomotion_GliderItem.AnimStateNode_2.Closed
  nodes [
   n1 = Get OwnerGlider()
   Output Animation Pose(Result=Blend Poses by bool(
     BlendPose_0=Sequence Player 'LocalZillaNight_SurfStyle_Closed_GLIDER'() {sequence=/BRCosmetics/Animation/Game/MainPlayer/Skydive/ParaGlide/LocalZillaNight/LocalZillaNight_SurfStyle_Closed_GLIDER}
     BlendPose_1=Sequence Player 'LocalZillaNight_SurfStyle_Closed_GLIDER'() {sequence=&LocalZillaNight_SurfStyle_Closed_GLIDER}
     bActiveValue=Get IsCosmeticPreview(self=@n1)
    ) {blendType=Linear})
  ]
 }
 Open_Into {
  path=AnimGraph.AnimGraphNode_StateMachine_2.Locomotion_GliderItem.AnimStateNode_3.Open_Into
  nodes [
   Apply Additive(bAlphaBoolEnabled=Get IsCosmeticPreview(self=Get OwnerGlider())) {alphaInputType=Bool, alphaBoolBlend {blendOption=Linear}}
   n1 = Break Fort Anim Input Skydiving(FortAnimInput_Skydiving=Get SkydivingState()) -> (
    ParaGilderRootOffset=~
    ParaGliderRootRotationOverride=~
    bUseParaGlideRootModifier=~
    bIsSkydivingFromLaunchPad=~
    bIsSkydivingFromBus=~
    bIsInVortex=~
    bIsDiving=~
    bIsParachuteOpen=~
    bIsSkydiving=~
    bIsParachuteLeaning=~
    bIsSkydiveLeaning=~
    bIsLeaning=~
    bIsSkydiveDiveMode=~
    bPlaySkydiveDrift=~
    bSkydiveDriftDelayActive=~
    bSkydiveDriftAnimAllowed=~
    bIsGliderRight=~
    bIsGliderCenter=~
    bIsGliderLeft=~
    bIsGliderForward=~
    bIsGliderBack=~
    bHasDeployedAndIsStandingUpright=~
    LocalAccelForward=~
    LocalAccelRight=~
    LocalVelocityRight=~
    SkydiveAimPitch=~
    SkydiveAimYaw=~
    DeployChuteAnimRate=~
    SkydiveDriftAnimRate=~
    SkydiveDriftAnimRateCurrent=~
    SkydiveFidgetAnimRate=~
    SkydiveFidgetAnimRateCurrent=~
    SkydiveAdditiveAlpha=~
    SkydiveDriftDelay=~
   )
   n2 = Blendspace Player 'BS_LocalZillaNight_SurfStyle_Into_GLIDER'(X=@n1.SkydiveAimYaw, Y=@n1.SkydiveAimPitch) {bLoop=false, blendSpace=&BS_LocalZillaNight_SurfStyle_Into_GLIDER}
   n3 = Sequence Player 'LocalZillaNight_SurfStyle_Into_FrontEnd_GLIDER'() {
    bLoopAnimation=false
    sequence=/BRCosmetics/Animation/Game/MainPlayer/Skydive/ParaGlide/LocalZillaNight/LocalZillaNight_SurfStyle_Into_FrontEnd_GLIDER
   }
   n4 = Get OwnerGlider()
   Output Animation Pose(Result=Blend Poses by bool(BlendPose_0=@n3, BlendPose_1=@n2, bActiveValue=Get IsCosmeticPreview(self=@n4)) {blendType=Linear})
  ]
 }
 Transition {
  path=AnimGraph.AnimGraphNode_StateMachine_2.Locomotion_GliderItem.AnimStateTransitionNode_0.Transition
  nodes [
   Result()
   "Equal (Float)"(A="Get Relevant Anim Time Remaining (Open_Into)"(), B=0)
  ]
 }
 Transition {
  path=AnimGraph.AnimGraphNode_StateMachine_2.Locomotion_GliderItem.AnimStateTransitionNode_1.Transition
  nodes [
   n1 = Break Fort Anim Input Skydiving(FortAnimInput_Skydiving=Get SkydivingState()) -> (
    ParaGilderRootOffset=~
    ParaGliderRootRotationOverride=~
    bUseParaGlideRootModifier=~
    bIsSkydivingFromLaunchPad=~
    bIsSkydivingFromBus=~
    bIsInVortex=~
    bIsDiving=~
    bIsParachuteOpen=~
    bIsSkydiving=~
    bIsParachuteLeaning=~
    bIsSkydiveLeaning=~
    bIsLeaning=~
    bIsSkydiveDiveMode=~
    bPlaySkydiveDrift=~
    bSkydiveDriftDelayActive=~
    bSkydiveDriftAnimAllowed=~
    bIsGliderRight=~
    bIsGliderCenter=~
    bIsGliderLeft=~
    bIsGliderForward=~
    bIsGliderBack=~
    bHasDeployedAndIsStandingUpright=~
    LocalAccelForward=~
    LocalAccelRight=~
    LocalVelocityRight=~
    SkydiveAimPitch=~
    SkydiveAimYaw=~
    DeployChuteAnimRate=~
    SkydiveDriftAnimRate=~
    SkydiveDriftAnimRateCurrent=~
    SkydiveFidgetAnimRate=~
    SkydiveFidgetAnimRateCurrent=~
    SkydiveAdditiveAlpha=~
    SkydiveDriftDelay=~
   )
   n2 = Get IsCosmeticPreview(self=Get OwnerGlider())
   Result(bCanEnterTransition=AND Boolean(
     A=OR Boolean!(A=@n1.bIsSkydivingFromLaunchPad, B=@n1.bIsSkydivingFromBus, C:bool=@n1.bIsInVortex, D:bool=@n2) -> (ReturnValue)
     B=@n1.bIsParachuteOpen
    ))
  ]
 }
 Transition {
  path=AnimGraph.AnimGraphNode_StateMachine_2.Locomotion_GliderItem.AnimStateTransitionNode_2.Transition
  nodes [
   n1 = Get bForceOpen()
   n2 = Break Fort Anim Input Skydiving(FortAnimInput_Skydiving=Get SkydivingState()) -> (
    ParaGilderRootOffset=~
    ParaGliderRootRotationOverride=~
    bUseParaGlideRootModifier=~
    bIsSkydivingFromLaunchPad=~
    bIsSkydivingFromBus=~
    bIsInVortex=~
    bIsDiving=~
    bIsParachuteOpen=~
    bIsSkydiving=~
    bIsParachuteLeaning=~
    bIsSkydiveLeaning=~
    bIsLeaning=~
    bIsSkydiveDiveMode=~
    bPlaySkydiveDrift=~
    bSkydiveDriftDelayActive=~
    bSkydiveDriftAnimAllowed=~
    bIsGliderRight=~
    bIsGliderCenter=~
    bIsGliderLeft=~
    bIsGliderForward=~
    bIsGliderBack=~
    bHasDeployedAndIsStandingUpright=~
    LocalAccelForward=~
    LocalAccelRight=~
    LocalVelocityRight=~
    SkydiveAimPitch=~
    SkydiveAimYaw=~
    DeployChuteAnimRate=~
    SkydiveDriftAnimRate=~
    SkydiveDriftAnimRateCurrent=~
    SkydiveFidgetAnimRate=~
    SkydiveFidgetAnimRateCurrent=~
    SkydiveAdditiveAlpha=~
    SkydiveDriftDelay=~
   )
   Result(bCanEnterTransition=AND Boolean(A=NOT Boolean(A=@n2.bIsParachuteOpen), B=NOT Boolean(A=@n1)))
  ]
 }
 Transition {
  path=AnimGraph.AnimGraphNode_StateMachine_2.Locomotion_GliderItem.AnimStateTransitionNode_3.Transition
  nodes [
   n1 = Get bForceOpen()
   n2 = Break Fort Anim Input Skydiving(FortAnimInput_Skydiving=Get SkydivingState()) -> (
    ParaGilderRootOffset=~
    ParaGliderRootRotationOverride=~
    bUseParaGlideRootModifier=~
    bIsSkydivingFromLaunchPad=~
    bIsSkydivingFromBus=~
    bIsInVortex=~
    bIsDiving=~
    bIsParachuteOpen=~
    bIsSkydiving=~
    bIsParachuteLeaning=~
    bIsSkydiveLeaning=~
    bIsLeaning=~
    bIsSkydiveDiveMode=~
    bPlaySkydiveDrift=~
    bSkydiveDriftDelayActive=~
    bSkydiveDriftAnimAllowed=~
    bIsGliderRight=~
    bIsGliderCenter=~
    bIsGliderLeft=~
    bIsGliderForward=~
    bIsGliderBack=~
    bHasDeployedAndIsStandingUpright=~
    LocalAccelForward=~
    LocalAccelRight=~
    LocalVelocityRight=~
    SkydiveAimPitch=~
    SkydiveAimYaw=~
    DeployChuteAnimRate=~
    SkydiveDriftAnimRate=~
    SkydiveDriftAnimRateCurrent=~
    SkydiveFidgetAnimRate=~
    SkydiveFidgetAnimRateCurrent=~
    SkydiveAdditiveAlpha=~
    SkydiveDriftDelay=~
   )
   Result(bCanEnterTransition=AND Boolean(A=NOT Boolean(A=@n2.bIsParachuteOpen), B=NOT Boolean(A=@n1)))
  ]
 }
 Transition {
  path=AnimGraph.AnimGraphNode_StateMachine_2.Locomotion_GliderItem.AnimStateTransitionNode_4.Transition
  nodes [
   n1 = Break Fort Anim Input Skydiving(FortAnimInput_Skydiving=Get SkydivingState()) -> (
    ParaGilderRootOffset=~
    ParaGliderRootRotationOverride=~
    bUseParaGlideRootModifier=~
    bIsSkydivingFromLaunchPad=~
    bIsSkydivingFromBus=~
    bIsInVortex=~
    bIsDiving=~
    bIsParachuteOpen=~
    bIsSkydiving=~
    bIsParachuteLeaning=~
    bIsSkydiveLeaning=~
    bIsLeaning=~
    bIsSkydiveDiveMode=~
    bPlaySkydiveDrift=~
    bSkydiveDriftDelayActive=~
    bSkydiveDriftAnimAllowed=~
    bIsGliderRight=~
    bIsGliderCenter=~
    bIsGliderLeft=~
    bIsGliderForward=~
    bIsGliderBack=~
    bHasDeployedAndIsStandingUpright=~
    LocalAccelForward=~
    LocalAccelRight=~
    LocalVelocityRight=~
    SkydiveAimPitch=~
    SkydiveAimYaw=~
    DeployChuteAnimRate=~
    SkydiveDriftAnimRate=~
    SkydiveDriftAnimRateCurrent=~
    SkydiveFidgetAnimRate=~
    SkydiveFidgetAnimRateCurrent=~
    SkydiveAdditiveAlpha=~
    SkydiveDriftDelay=~
   )
   Result(bCanEnterTransition=AND Boolean!(
     A=NOT Boolean(A=@n1.bIsSkydivingFromLaunchPad)
     B=NOT Boolean(A=@n1.bIsSkydivingFromBus)
     C:bool=NOT Boolean(A=@n1.bIsInVortex)
     D:bool=Get bForceOpen()
    ) -> (ReturnValue))
  ]
 }
 Transition {path=AnimGraph.AnimGraphNode_StateMachine_2.Locomotion_GliderItem.AnimStateTransitionNode_5.Transition, nodes [Result()]}
 Transition {
  path=AnimGraph.AnimGraphNode_StateMachine_2.Locomotion_GliderItem.AnimStateTransitionNode_6.Transition
  nodes [
   n1 = Get bForceOpen()
   n2 = Break Fort Anim Input Skydiving(FortAnimInput_Skydiving=Get SkydivingState()) -> (
    ParaGilderRootOffset=~
    ParaGliderRootRotationOverride=~
    bUseParaGlideRootModifier=~
    bIsSkydivingFromLaunchPad=~
    bIsSkydivingFromBus=~
    bIsInVortex=~
    bIsDiving=~
    bIsParachuteOpen=~
    bIsSkydiving=~
    bIsParachuteLeaning=~
    bIsSkydiveLeaning=~
    bIsLeaning=~
    bIsSkydiveDiveMode=~
    bPlaySkydiveDrift=~
    bSkydiveDriftDelayActive=~
    bSkydiveDriftAnimAllowed=~
    bIsGliderRight=~
    bIsGliderCenter=~
    bIsGliderLeft=~
    bIsGliderForward=~
    bIsGliderBack=~
    bHasDeployedAndIsStandingUpright=~
    LocalAccelForward=~
    LocalAccelRight=~
    LocalVelocityRight=~
    SkydiveAimPitch=~
    SkydiveAimYaw=~
    DeployChuteAnimRate=~
    SkydiveDriftAnimRate=~
    SkydiveDriftAnimRateCurrent=~
    SkydiveFidgetAnimRate=~
    SkydiveFidgetAnimRateCurrent=~
    SkydiveAdditiveAlpha=~
    SkydiveDriftDelay=~
   )
   Result(bCanEnterTransition=AND Boolean!(A=NOT Boolean(A=@n2.bIsParachuteOpen), B=NOT Boolean(A=@n1)) -> (ReturnValue))
  ]
 }
 Head Darts {
  entry=Identify
  path=AnimGraph.AnimGraphNode_StateMachine_3.Head Darts
  ids=id
  nodes [
   To Darts() {is=title AnimStateAliasNode, also {aliases [Identify]}}
   To Identify() {is=title AnimStateAliasNode, also {aliases [Left HeadDart, Right HeadDart]}}
   AnimStateNode_0 = Identify() {is=title AnimStateNode}
   AnimStateNode_2 = Right HeadDart() {is=title AnimStateNode}
   AnimStateNode_3 = Left HeadDart() {is=title AnimStateNode}
   AnimStateTransitionNode_2 = AnimStateTransitionNode() {bAutomaticRuleBasedOnSequencePlayerInState=true}
   AnimStateTransitionNode_5 = AnimStateTransitionNode()
   AnimStateTransitionNode_6 = AnimStateTransitionNode()
  ]
  links [
   {
    dst=Left HeadDart
    rule=AnimGraph.AnimGraphNode_StateMachine_3.Head Darts.AnimStateTransitionNode_6.Transition
    src=To Darts
    via=AnimStateTransitionNode_6
   }
   {
    dst=Right HeadDart
    rule=AnimGraph.AnimGraphNode_StateMachine_3.Head Darts.AnimStateTransitionNode_5.Transition
    src=To Darts
    via=AnimStateTransitionNode_5
   }
   {
    dst=Identify
    rule=AnimGraph.AnimGraphNode_StateMachine_3.Head Darts.AnimStateTransitionNode_2.Transition
    src=To Identify
    via=AnimStateTransitionNode_2
   }
  ]
 }
 Identify {
  path=AnimGraph.AnimGraphNode_StateMachine_3.Head Darts.AnimStateNode_0.Identify
  nodes [Output Animation Pose(Result=Additive Identity Pose())]
 }
 Right HeadDart {
  path=AnimGraph.AnimGraphNode_StateMachine_3.Head Darts.AnimStateNode_2.Right HeadDart
  nodes [
   n1 = "Sequence Player (additive) 'CoconutShell_Right_HeadDarts_Anim_GLIDER'"() {
    bLoopAnimation=false
    sequence=&/CoconutShell_Right_HeadDarts_Anim_GLIDER
   }
   Output Animation Pose(Result=@n1)
  ]
 }
 Left HeadDart {
  path=AnimGraph.AnimGraphNode_StateMachine_3.Head Darts.AnimStateNode_3.Left HeadDart
  nodes [
   n1 = "Sequence Player (additive) 'CoconutShell_Left_HeadDarts_Anim_GLIDER'"() {bLoopAnimation=false, sequence=&/CoconutShell_Left_HeadDarts_Anim_GLIDER}
   Output Animation Pose(Result=@n1)
  ]
 }
 Transition {path=AnimGraph.AnimGraphNode_StateMachine_3.Head Darts.AnimStateTransitionNode_2.Transition, nodes [Result()]}
 Transition {
  path=AnimGraph.AnimGraphNode_StateMachine_3.Head Darts.AnimStateTransitionNode_5.Transition
  nodes [Result(bCanEnterTransition=Get RightHeadDart() -> (RightHeadDart:bool))]
 }
 Transition {
  path=AnimGraph.AnimGraphNode_StateMachine_3.Head Darts.AnimStateTransitionNode_6.Transition
  nodes [Result(bCanEnterTransition=Get LeftHeadDart() -> (LeftHeadDart:bool))]
 }
 Blinks {
  entry=Identify
  path=AnimGraph.AnimGraphNode_StateMachine_4.Blinks
  ids=id
  nodes [
   To Identify() {also {aliases [Blink 1]}}
   To Blinks() {is=title AnimStateAliasNode, also {aliases [Identify]}}
   AnimStateNode_0 = Identify()
   AnimStateNode_2 = Blink 1() {is=title AnimStateNode}
   AnimStateTransitionNode_8 = AnimStateTransitionNode() {bAutomaticRuleBasedOnSequencePlayerInState=true}
   AnimStateTransitionNode_9 = AnimStateTransitionNode() {crossfadeDuration=0.1}
  ]
  links [
   {
    dst=Blink 1
    rule=AnimGraph.AnimGraphNode_StateMachine_4.Blinks.AnimStateTransitionNode_9.Transition
    src=To Blinks
    via=AnimStateTransitionNode_9
   }
   {
    dst=Identify
    rule=AnimGraph.AnimGraphNode_StateMachine_4.Blinks.AnimStateTransitionNode_8.Transition
    src=To Identify
    via=AnimStateTransitionNode_8
   }
  ]
 }
 Identify {path=AnimGraph.AnimGraphNode_StateMachine_4.Blinks.AnimStateNode_0.Identify, nodes [Output Animation Pose(Result=Additive Identity Pose())]}
 Blink 1 {
  path=AnimGraph.AnimGraphNode_StateMachine_4.Blinks.AnimStateNode_2.Blink 1
  nodes [Output Animation Pose(Result="Sequence Player (additive) 'CoconutShell_Blink_Anim_GLIDER'"() {bLoopAnimation=false, sequence=&/CoconutShell_Blink_Anim_GLIDER})]
 }
 Transition {path=AnimGraph.AnimGraphNode_StateMachine_4.Blinks.AnimStateTransitionNode_8.Transition, nodes [Result()]}
 Transition {
  path=AnimGraph.AnimGraphNode_StateMachine_4.Blinks.AnimStateTransitionNode_9.Transition
  nodes [Result(bCanEnterTransition=Get FirstBlink() -> (FirstBlink:bool))]
 }
 Caw {
  entry=Identify
  path=AnimGraph.AnimGraphNode_StateMachine_5.Caw
  ids=id
  nodes [
   AnimStateNode_0 = Identify()
   AnimStateNode_1 = Caw() {is=title AnimStateNode}
   AnimStateTransitionNode_0 = AnimStateTransitionNode()
   AnimStateTransitionNode_1 = AnimStateTransitionNode() {bAutomaticRuleBasedOnSequencePlayerInState=true}
  ]
  links [
   {dst=Identify, rule=AnimGraph.AnimGraphNode_StateMachine_5.Caw.AnimStateTransitionNode_1.Transition, src=Caw, via=AnimStateTransitionNode_1}
   {dst=Caw, rule=AnimGraph.AnimGraphNode_StateMachine_5.Caw.AnimStateTransitionNode_0.Transition, src=Identify, via=AnimStateTransitionNode_0}
  ]
 }
 Identify {path=AnimGraph.AnimGraphNode_StateMachine_5.Caw.AnimStateNode_0.Identify, nodes [Output Animation Pose(Result=Additive Identity Pose())]}
 Caw {
  path=AnimGraph.AnimGraphNode_StateMachine_5.Caw.AnimStateNode_1.Caw
  nodes [Output Animation Pose(Result="Sequence Player (additive) 'CoconutShell_Caw_Anim_GLIDER'"() {bLoopAnimation=false, sequence=&/CoconutShell_Caw_Anim_GLIDER})]
 }
 Transition {
  path=AnimGraph.AnimGraphNode_StateMachine_5.Caw.AnimStateTransitionNode_0.Transition
  nodes [Result(bCanEnterTransition=Get Caw() -> (Caw:bool))]
 }
 Transition {path=AnimGraph.AnimGraphNode_StateMachine_5.Caw.AnimStateTransitionNode_1.Transition, nodes [Result()]}
 Flaps {
  entry=Identify
  path=AnimGraph.AnimGraphNode_StateMachine_6.Flaps
  ids=id
  nodes [
   to Identify() {is=title AnimStateAliasNode, also {aliases [Flap Back, Flap Forward, One Flap, Two Flaps]}}
   To One Flap() {is=title AnimStateAliasNode, also {aliases [Identify]}}
   To Flap Forward() {is=title AnimStateAliasNode, also {aliases [Identify]}}
   To Two Flaps() {is=title AnimStateAliasNode, also {aliases [Identify]}}
   To Back Flap() {is=title AnimStateAliasNode, also {aliases [Identify]}}
   AnimStateNode_0 = Identify()
   AnimStateNode_1 = Flap Forward() {is=title AnimStateNode}
   AnimStateNode_2 = One Flap() {is=title AnimStateNode}
   AnimStateNode_3 = Two Flaps() {is=title AnimStateNode}
   AnimStateNode_4 = Flap Back() {is=title AnimStateNode}
   AnimStateTransitionNode_1 = AnimStateTransitionNode() {bAutomaticRuleBasedOnSequencePlayerInState=true}
   AnimStateTransitionNode_11 = AnimStateTransitionNode()
   AnimStateTransitionNode_12 = AnimStateTransitionNode()
   AnimStateTransitionNode_13 = AnimStateTransitionNode()
   AnimStateTransitionNode_2 = AnimStateTransitionNode()
  ]
  links [
   {
    dst=Flap Back
    rule=AnimGraph.AnimGraphNode_StateMachine_6.Flaps.AnimStateTransitionNode_13.Transition
    src=To Back Flap
    via=AnimStateTransitionNode_13
   }
   {
    dst=Flap Forward
    rule=AnimGraph.AnimGraphNode_StateMachine_6.Flaps.AnimStateTransitionNode_11.Transition
    src=To Flap Forward
    via=AnimStateTransitionNode_11
   }
   {
    dst=One Flap
    rule=AnimGraph.AnimGraphNode_StateMachine_6.Flaps.AnimStateTransitionNode_2.Transition
    src=To One Flap
    via=AnimStateTransitionNode_2
   }
   {
    dst=Two Flaps
    rule=AnimGraph.AnimGraphNode_StateMachine_6.Flaps.AnimStateTransitionNode_12.Transition
    src=To Two Flaps
    via=AnimStateTransitionNode_12
   }
   {
    dst=Identify
    rule=AnimGraph.AnimGraphNode_StateMachine_6.Flaps.AnimStateTransitionNode_1.Transition
    src=to Identify
    via=AnimStateTransitionNode_1
   }
  ]
 }
 Identify {path=AnimGraph.AnimGraphNode_StateMachine_6.Flaps.AnimStateNode_0.Identify, nodes [Output Animation Pose(Result=Additive Identity Pose())]}
 Flap Forward {
  path=AnimGraph.AnimGraphNode_StateMachine_6.Flaps.AnimStateNode_1.Flap Forward
  nodes [
   n1 = "Sequence Player (additive) 'CoconutShell_Flaps_Forward_Anim_GLIDER'"() {bLoopAnimation=false, sequence=&/CoconutShell_Flaps_Forward_Anim_GLIDER}
   Output Animation Pose(Result=@n1)
  ]
 }
 One Flap {
  path=AnimGraph.AnimGraphNode_StateMachine_6.Flaps.AnimStateNode_2.One Flap
  nodes [Output Animation Pose(Result="Sequence Player (additive) 'CoconutShell_Flaps_Anim_GLIDER'"() {bLoopAnimation=false, sequence=&/CoconutShell_Flaps_Anim_GLIDER})]
 }
 Two Flaps {
  path=AnimGraph.AnimGraphNode_StateMachine_6.Flaps.AnimStateNode_3.Two Flaps
  nodes [
   n1 = "Sequence Player (additive) 'CoconutShell_Flaps2x_Anim_GLIDER'"() {bLoopAnimation=false, sequence=&/CoconutShell_Flaps2x_Anim_GLIDER}
   Output Animation Pose(Result=@n1)
  ]
 }
 Flap Back {
  path=AnimGraph.AnimGraphNode_StateMachine_6.Flaps.AnimStateNode_4.Flap Back
  nodes [
   n1 = "Sequence Player (additive) 'CoconutShell_Flaps_Back_Anim_GLIDER'"() {bLoopAnimation=false, sequence=&/CoconutShell_Flaps_Back_Anim_GLIDER}
   Output Animation Pose(Result=@n1)
  ]
 }
 Transition {path=AnimGraph.AnimGraphNode_StateMachine_6.Flaps.AnimStateTransitionNode_1.Transition, nodes [Result()]}
 Transition {
  path=AnimGraph.AnimGraphNode_StateMachine_6.Flaps.AnimStateTransitionNode_11.Transition
  nodes [Result(bCanEnterTransition=Get FlapForward() -> (FlapForward:bool))]
 }
 Transition {
  path=AnimGraph.AnimGraphNode_StateMachine_6.Flaps.AnimStateTransitionNode_12.Transition
  nodes [Result(bCanEnterTransition=Get TwoFlaps() -> (TwoFlaps:bool))]
 }
 Transition {
  path=AnimGraph.AnimGraphNode_StateMachine_6.Flaps.AnimStateTransitionNode_13.Transition
  nodes [Result(bCanEnterTransition=Get FlapBack() -> (FlapBack:bool))]
 }
 Transition {
  path=AnimGraph.AnimGraphNode_StateMachine_6.Flaps.AnimStateTransitionNode_2.Transition
  nodes [Result(bCanEnterTransition=Get OneFlap() -> (OneFlap:bool))]
 }
 LeanTransitions {
  entry=Lean_Zero
  path=AnimGraph.AnimGraphNode_StateMachine_8.LeanTransitions
  ids=id
  nodes [
   AnimStateNode_10 = Lean_Zero() {
    is=title AnimStateNode
    stateEntered {guid=#4}
    stateFullyBlended {guid=#5}
    stateLeft {guid=#6}
   }
   AnimStateNode_14 = Lean_Out() {
    is=title AnimStateNode
    stateEntered {guid=#7}
    stateFullyBlended {guid=#8}
    stateLeft {guid=#9}
   }
   AnimStateNode_17 = Lean_Into() {
    is=title AnimStateNode
    stateEntered {guid=#10}
    stateFullyBlended {guid=#11}
    stateLeft {guid=#12}
   }
   AnimStateTransitionNode_26 = AnimStateTransitionNode() {bAutomaticRuleBasedOnSequencePlayerInState=true}
   AnimStateTransitionNode_37 = AnimStateTransitionNode()
   AnimStateTransitionNode_38 = AnimStateTransitionNode()
   AnimStateTransitionNode_40 = AnimStateTransitionNode()
  ]
  links [
   {
    dst=Lean_Out
    rule=AnimGraph.AnimGraphNode_StateMachine_8.LeanTransitions.AnimStateTransitionNode_38.AnimationTransitionGraph_0
    src=Lean_Into
    via=AnimStateTransitionNode_38
   }
   {
    dst=Lean_Into
    rule=AnimGraph.AnimGraphNode_StateMachine_8.LeanTransitions.AnimStateTransitionNode_40.AnimationTransitionGraph_0
    src=Lean_Out
    via=AnimStateTransitionNode_40
   }
   {
    dst=Lean_Zero
    rule=AnimGraph.AnimGraphNode_StateMachine_8.LeanTransitions.AnimStateTransitionNode_26.AnimationTransitionGraph_0
    src=Lean_Out
    via=AnimStateTransitionNode_26
   }
   {
    dst=Lean_Into
    rule=AnimGraph.AnimGraphNode_StateMachine_8.LeanTransitions.AnimStateTransitionNode_37.AnimationTransitionGraph_0
    src=Lean_Zero
    via=AnimStateTransitionNode_37
   }
  ]
 }
 Lean_Zero {
  path=AnimGraph.AnimGraphNode_StateMachine_8.LeanTransitions.AnimStateNode_10.Lean_Zero
  nodes [Output Animation Pose(Result="Sequence Player (additive) 'DoubleDuty_SurfStyle_Lean_Center_Pose_GLIDER'"() {sequence=/BRCosmetics/Animation/Game/MainPlayer/Skydive/ParaGlide/DoubleDuty/DoubleDuty_SurfStyle_Lean_Center_Pose_GLIDER})]
 }
 Lean_Out {
  path=AnimGraph.AnimGraphNode_StateMachine_8.LeanTransitions.AnimStateNode_14.Lean_Out
  nodes [
   AnimGraphNode_StateMachine_0 = Lean_Out() {is=title AnimGraphNode_StateMachine}
   n1 = "Sequence Player (additive) 'CoconutShell_Lean_Center_Pose_GLIDER'"() {bLoopAnimation=false, sequence=&/CoconutShell_Lean_Center_Pose_GLIDER}
   n2 = "Sequence Player (additive) 'CoconutShell_Lean_Forward_Out_GLIDER'"() {bLoopAnimation=false, sequence=&/CoconutShell_Lean_Forward_Out_GLIDER}
   n3 = "Sequence Player (additive) 'CoconutShell_Lean_Back_Out_GLIDER'"() {bLoopAnimation=false, sequence=&/CoconutShell_Lean_Back_Out_GLIDER}
   n4 = "Sequence Player (additive) 'CoconutShell_Lean_Right_Out_GLIDER'"() {bLoopAnimation=false, sequence=&/CoconutShell_Lean_Right_Out_GLIDER}
   n5 = "Sequence Player (additive) 'CoconutShell_Lean_Left_Out_GLIDER'"() {bLoopAnimation=false, sequence=&/CoconutShell_Lean_Left_Out_GLIDER}
   n6 = "Blend Poses (ESkydivingDirection)"!(
    BlendPose_0:Pose Link Structure=@n1
    BlendPose_1:Pose Link Structure=@n2
    BlendPose_2:Pose Link Structure=@n3
    BlendPose_3:Pose Link Structure=@n4
    BlendPose_4:Pose Link Structure=@n5
    BlendTime_0:float=1
    BlendTime_1:float=1
    BlendTime_2:float=1
    BlendTime_3:float=1
    BlendTime_4:float=1
    ActiveEnumValue:Byte=Break Fort Anim Input Skydiving(FortAnimInput_Skydiving=Get SkydivingState()) -> (
     ParaGilderRootOffset=~
     ParaGliderRootRotationOverride=~
     bUseParaGlideRootModifier=~
     bIsSkydivingFromLaunchPad=~
     bIsSkydivingFromBus=~
     bIsInVortex=~
     bIsDiving=~
     bIsParachuteOpen=~
     bIsSkydiving=~
     bIsParachuteLeaning=~
     bIsSkydiveLeaning=~
     bIsLeaning=~
     bIsSkydiveDiveMode=~
     bPlaySkydiveDrift=~
     bSkydiveDriftDelayActive=~
     bSkydiveDriftAnimAllowed=~
     bIsGliderRight=~
     bIsGliderCenter=~
     bIsGliderLeft=~
     bIsGliderForward=~
     bIsGliderBack=~
     bHasDeployedAndIsStandingUpright=~
     LocalAccelForward=~
     LocalAccelRight=~
     LocalVelocityRight=~
     SkydiveAimPitch=~
     SkydiveAimYaw=~
     DeployChuteAnimRate=~
     SkydiveDriftAnimRate=~
     SkydiveDriftAnimRateCurrent=~
     SkydiveFidgetAnimRate=~
     SkydiveFidgetAnimRateCurrent=~
     SkydiveAdditiveAlpha=~
     SkydiveDriftDelay=~
    ).DirectionLast
   ) -> (Pose:Pose Link Structure) {
    is=title AnimGraphNode_BlendListByEnum
    blendType=Linear
    blendPose [
     {linkId=-1, sourceLinkId=-1}
     {linkId=-1, sourceLinkId=-1}
     {linkId=-1, sourceLinkId=-1}
     {linkId=-1, sourceLinkId=-1}
     {linkId=-1, sourceLinkId=-1}
    ]
    blendTime [0.1,0.1,0.1,0.1,0.1]
   }
   Output Animation Pose(Result=@n6)
  ]
 }
 Lean_Out {
  entry=Lean_Zero
  path=AnimGraph.AnimGraphNode_StateMachine_8.LeanTransitions.AnimStateNode_14.Lean_Out.AnimGraphNode_StateMachine_0.Lean_Out
  ids=id
  nodes [
   AnimStateNode_0 = Lean_Zero()
   AnimStateNode_3 = Left_Out() {is=title AnimStateNode}
   AnimStateNode_4 = Right_Out() {is=title AnimStateNode}
   AnimStateNode_5 = Forward_Out() {is=title AnimStateNode}
   AnimStateNode_6 = Back_Out() {is=title AnimStateNode}
   AnimStateTransitionNode_19 = AnimStateTransitionNode()
   AnimStateTransitionNode_20 = AnimStateTransitionNode()
   AnimStateTransitionNode_21 = AnimStateTransitionNode()
   AnimStateTransitionNode_22 = AnimStateTransitionNode()
   AnimStateTransitionNode_23 = AnimStateTransitionNode() {crossfadeDuration=0.35}
   AnimStateTransitionNode_24 = AnimStateTransitionNode() {crossfadeDuration=0.35}
   AnimStateTransitionNode_25 = AnimStateTransitionNode() {crossfadeDuration=0.5}
   AnimStateTransitionNode_26 = AnimStateTransitionNode() {crossfadeDuration=0.5}
   AnimStateTransitionNode_27 = AnimStateTransitionNode() {crossfadeDuration=0.5}
   AnimStateTransitionNode_28 = AnimStateTransitionNode() {crossfadeDuration=0.5}
   AnimStateTransitionNode_29 = AnimStateTransitionNode() {crossfadeDuration=0.5}
   AnimStateTransitionNode_30 = AnimStateTransitionNode() {crossfadeDuration=0.5}
   AnimStateTransitionNode_31 = AnimStateTransitionNode() {crossfadeDuration=0.5}
   AnimStateTransitionNode_32 = AnimStateTransitionNode() {crossfadeDuration=0.5}
   AnimStateTransitionNode_33 = AnimStateTransitionNode() {crossfadeDuration=0.5}
   AnimStateTransitionNode_34 = AnimStateTransitionNode() {crossfadeDuration=0.5}
  ]
  links [
   {
    dst=Forward_Out
    rule=AnimGraph.AnimGraphNode_StateMachine_8.LeanTransitions.AnimStateNode_14.Lean_Out.AnimGraphNode_StateMachine_0.Lean_Out.AnimStateTransitionNode_32.AnimationTransitionGraph_1
    src=Back_Out
    via=AnimStateTransitionNode_32
   }
   {
    dst=Left_Out
    rule=AnimGraph.AnimGraphNode_StateMachine_8.LeanTransitions.AnimStateNode_14.Lean_Out.AnimGraphNode_StateMachine_0.Lean_Out.AnimStateTransitionNode_33.AnimationTransitionGraph_1
    src=Back_Out
    via=AnimStateTransitionNode_33
   }
   {
    dst=Right_Out
    rule=AnimGraph.AnimGraphNode_StateMachine_8.LeanTransitions.AnimStateNode_14.Lean_Out.AnimGraphNode_StateMachine_0.Lean_Out.AnimStateTransitionNode_34.AnimationTransitionGraph_1
    src=Back_Out
    via=AnimStateTransitionNode_34
   }
   {
    dst=Back_Out
    rule=AnimGraph.AnimGraphNode_StateMachine_8.LeanTransitions.AnimStateNode_14.Lean_Out.AnimGraphNode_StateMachine_0.Lean_Out.AnimStateTransitionNode_31.AnimationTransitionGraph_1
    src=Forward_Out
    via=AnimStateTransitionNode_31
   }
   {
    dst=Left_Out
    rule=AnimGraph.AnimGraphNode_StateMachine_8.LeanTransitions.AnimStateNode_14.Lean_Out.AnimGraphNode_StateMachine_0.Lean_Out.AnimStateTransitionNode_29.AnimationTransitionGraph_1
    src=Forward_Out
    via=AnimStateTransitionNode_29
   }
   {
    dst=Right_Out
    rule=AnimGraph.AnimGraphNode_StateMachine_8.LeanTransitions.AnimStateNode_14.Lean_Out.AnimGraphNode_StateMachine_0.Lean_Out.AnimStateTransitionNode_30.AnimationTransitionGraph_1
    src=Forward_Out
    via=AnimStateTransitionNode_30
   }
   {
    dst=Back_Out
    rule=AnimGraph.AnimGraphNode_StateMachine_8.LeanTransitions.AnimStateNode_14.Lean_Out.AnimGraphNode_StateMachine_0.Lean_Out.AnimStateTransitionNode_20.AnimationTransitionGraph_1
    src=Lean_Zero
    via=AnimStateTransitionNode_20
   }
   {
    dst=Forward_Out
    rule=AnimGraph.AnimGraphNode_StateMachine_8.LeanTransitions.AnimStateNode_14.Lean_Out.AnimGraphNode_StateMachine_0.Lean_Out.AnimStateTransitionNode_19.AnimationTransitionGraph_1
    src=Lean_Zero
    via=AnimStateTransitionNode_19
   }
   {
    dst=Left_Out
    rule=AnimGraph.AnimGraphNode_StateMachine_8.LeanTransitions.AnimStateNode_14.Lean_Out.AnimGraphNode_StateMachine_0.Lean_Out.AnimStateTransitionNode_22.AnimationTransitionGraph_1
    src=Lean_Zero
    via=AnimStateTransitionNode_22
   }
   {
    dst=Right_Out
    rule=AnimGraph.AnimGraphNode_StateMachine_8.LeanTransitions.AnimStateNode_14.Lean_Out.AnimGraphNode_StateMachine_0.Lean_Out.AnimStateTransitionNode_21.AnimationTransitionGraph_1
    src=Lean_Zero
    via=AnimStateTransitionNode_21
   }
   {
    dst=Back_Out
    rule=AnimGraph.AnimGraphNode_StateMachine_8.LeanTransitions.AnimStateNode_14.Lean_Out.AnimGraphNode_StateMachine_0.Lean_Out.AnimStateTransitionNode_26.AnimationTransitionGraph_1
    src=Left_Out
    via=AnimStateTransitionNode_26
   }
   {
    dst=Forward_Out
    rule=AnimGraph.AnimGraphNode_StateMachine_8.LeanTransitions.AnimStateNode_14.Lean_Out.AnimGraphNode_StateMachine_0.Lean_Out.AnimStateTransitionNode_25.AnimationTransitionGraph_1
    src=Left_Out
    via=AnimStateTransitionNode_25
   }
   {
    dst=Right_Out
    rule=AnimGraph.AnimGraphNode_StateMachine_8.LeanTransitions.AnimStateNode_14.Lean_Out.AnimGraphNode_StateMachine_0.Lean_Out.AnimStateTransitionNode_24.AnimationTransitionGraph_1
    src=Left_Out
    via=AnimStateTransitionNode_24
   }
   {
    dst=Back_Out
    rule=AnimGraph.AnimGraphNode_StateMachine_8.LeanTransitions.AnimStateNode_14.Lean_Out.AnimGraphNode_StateMachine_0.Lean_Out.AnimStateTransitionNode_28.AnimationTransitionGraph_1
    src=Right_Out
    via=AnimStateTransitionNode_28
   }
   {
    dst=Forward_Out
    rule=AnimGraph.AnimGraphNode_StateMachine_8.LeanTransitions.AnimStateNode_14.Lean_Out.AnimGraphNode_StateMachine_0.Lean_Out.AnimStateTransitionNode_27.AnimationTransitionGraph_1
    src=Right_Out
    via=AnimStateTransitionNode_27
   }
   {
    dst=Left_Out
    rule=AnimGraph.AnimGraphNode_StateMachine_8.LeanTransitions.AnimStateNode_14.Lean_Out.AnimGraphNode_StateMachine_0.Lean_Out.AnimStateTransitionNode_23.AnimationTransitionGraph_1
    src=Right_Out
    via=AnimStateTransitionNode_23
   }
  ]
 }
 Lean_Zero {
  path=AnimGraph.AnimGraphNode_StateMachine_8.LeanTransitions.AnimStateNode_14.Lean_Out.AnimGraphNode_StateMachine_0.Lean_Out.AnimStateNode_0.Lean_Zero
  nodes [Output Animation Pose(Result="Sequence Player (additive) 'DoubleDuty_SurfStyle_Lean_Center_Pose_GLIDER'"() {sequence=&DoubleDuty_SurfStyle_Lean_Center_Pose_GLIDER})]
 }
 Left_Out {
  path=AnimGraph.AnimGraphNode_StateMachine_8.LeanTransitions.AnimStateNode_14.Lean_Out.AnimGraphNode_StateMachine_0.Lean_Out.AnimStateNode_3.Left_Out
  nodes [Output Animation Pose(Result="Sequence Player (additive) 'DoubleDuty_Lean_Left_Out_GLIDER'"() {bLoopAnimation=false, sequence=/BRCosmetics/Animation/Game/MainPlayer/Skydive/ParaGlide/DoubleDuty/DoubleDuty_Lean_Left_Out_GLIDER})]
 }
 Right_Out {
  path=AnimGraph.AnimGraphNode_StateMachine_8.LeanTransitions.AnimStateNode_14.Lean_Out.AnimGraphNode_StateMachine_0.Lean_Out.AnimStateNode_4.Right_Out
  nodes [Output Animation Pose(Result="Sequence Player (additive) 'DoubleDuty_Lean_Right_Out_GLIDER'"() {sequence=/BRCosmetics/Animation/Game/MainPlayer/Skydive/ParaGlide/DoubleDuty/DoubleDuty_Lean_Right_Out_GLIDER})]
 }
 Forward_Out {
  path=AnimGraph.AnimGraphNode_StateMachine_8.LeanTransitions.AnimStateNode_14.Lean_Out.AnimGraphNode_StateMachine_0.Lean_Out.AnimStateNode_5.Forward_Out
  nodes [Output Animation Pose(Result="Sequence Player (additive) 'DoubleDuty_Lean_Forward_Out_GLIDER'"() {sequence=/BRCosmetics/Animation/Game/MainPlayer/Skydive/ParaGlide/DoubleDuty/DoubleDuty_Lean_Forward_Out_GLIDER})]
 }
 Back_Out {
  path=AnimGraph.AnimGraphNode_StateMachine_8.LeanTransitions.AnimStateNode_14.Lean_Out.AnimGraphNode_StateMachine_0.Lean_Out.AnimStateNode_6.Back_Out
  nodes [Output Animation Pose(Result="Sequence Player (additive) 'DoubleDuty_Lean_Back_Out_GLIDER'"() {sequence=/BRCosmetics/Animation/Game/MainPlayer/Skydive/ParaGlide/DoubleDuty/DoubleDuty_Lean_Back_Out_GLIDER})]
 }
 AnimationTransitionGraph_1 {
  path=AnimGraph.AnimGraphNode_StateMachine_8.LeanTransitions.AnimStateNode_14.Lean_Out.AnimGraphNode_StateMachine_0.Lean_Out.AnimStateTransitionNode_19.AnimationTransitionGraph_1
  nodes [Result(bCanEnterTransition=Break Fort Anim Input Skydiving!(FortAnimInput_Skydiving=Get SkydivingState()) -> (
     bIsGliderRight=~
     bIsGliderCenter=~
     bIsGliderLeft=~
     bIsGliderForward=~
     bIsGliderBack=~
     bHasDeployedAndIsStandingUpright=~
     LocalAccelDir
     PoseAdditive
    ).bIsGliderForward)]
 }
 AnimationTransitionGraph_1 {
  path=AnimGraph.AnimGraphNode_StateMachine_8.LeanTransitions.AnimStateNode_14.Lean_Out.AnimGraphNode_StateMachine_0.Lean_Out.AnimStateTransitionNode_20.AnimationTransitionGraph_1
  nodes [Result(bCanEnterTransition=Break Fort Anim Input Skydiving(FortAnimInput_Skydiving=Get SkydivingState()) -> (bIsGliderRight=~, bIsGliderCenter=~, bIsGliderLeft=~, bIsGliderForward=~, bIsGliderBack=~, bHasDeployedAndIsStandingUpright=~).bIsGliderBack)]
 }
 AnimationTransitionGraph_1 {
  path=AnimGraph.AnimGraphNode_StateMachine_8.LeanTransitions.AnimStateNode_14.Lean_Out.AnimGraphNode_StateMachine_0.Lean_Out.AnimStateTransitionNode_21.AnimationTransitionGraph_1
  nodes [
   n1 = Break Fort Anim Input Skydiving(FortAnimInput_Skydiving=Get SkydivingState()) -> (bIsGliderRight=~, bIsGliderCenter=~, bIsGliderLeft=~, bIsGliderForward=~, bIsGliderBack=~, bHasDeployedAndIsStandingUpright=~)
   Result(bCanEnterTransition=@n1)
   "Equal (Enum)"(A:ESkydivingDirection Enum=@n1.LocalAccelDir, B:ESkydivingDirection Enum=Right)
  ]
 }
 AnimationTransitionGraph_1 {
  path=AnimGraph.AnimGraphNode_StateMachine_8.LeanTransitions.AnimStateNode_14.Lean_Out.AnimGraphNode_StateMachine_0.Lean_Out.AnimStateTransitionNode_22.AnimationTransitionGraph_1
  nodes [Result(bCanEnterTransition=Break Fort Anim Input Skydiving(FortAnimInput_Skydiving=Get SkydivingState()) -> (bIsGliderRight=~, bIsGliderCenter=~, bIsGliderLeft=~, bIsGliderForward=~, bIsGliderBack=~, bHasDeployedAndIsStandingUpright=~).bIsGliderLeft)]
 }
 AnimationTransitionGraph_1 {
  path=AnimGraph.AnimGraphNode_StateMachine_8.LeanTransitions.AnimStateNode_14.Lean_Out.AnimGraphNode_StateMachine_0.Lean_Out.AnimStateTransitionNode_23.AnimationTransitionGraph_1
  nodes [Result(bCanEnterTransition=Break Fort Anim Input Skydiving(FortAnimInput_Skydiving=Get SkydivingState()) -> (bIsGliderRight=~, bIsGliderCenter=~, bIsGliderLeft=~, bIsGliderForward=~, bIsGliderBack=~, bHasDeployedAndIsStandingUpright=~).bIsGliderLeft)]
 }
 AnimationTransitionGraph_1 {
  path=AnimGraph.AnimGraphNode_StateMachine_8.LeanTransitions.AnimStateNode_14.Lean_Out.AnimGraphNode_StateMachine_0.Lean_Out.AnimStateTransitionNode_24.AnimationTransitionGraph_1
  nodes [
   n1 = Break Fort Anim Input Skydiving(FortAnimInput_Skydiving=Get SkydivingState()) -> (bIsGliderRight=~, bIsGliderCenter=~, bIsGliderLeft=~, bIsGliderForward=~, bIsGliderBack=~, bHasDeployedAndIsStandingUpright=~)
   Result(bCanEnterTransition=@n1)
   "Equal (Enum)"(A=@n1.LocalAccelDir, B=Right)
  ]
 }
 AnimationTransitionGraph_1 {
  path=AnimGraph.AnimGraphNode_StateMachine_8.LeanTransitions.AnimStateNode_14.Lean_Out.AnimGraphNode_StateMachine_0.Lean_Out.AnimStateTransitionNode_25.AnimationTransitionGraph_1
  nodes [Result(bCanEnterTransition=Break Fort Anim Input Skydiving(FortAnimInput_Skydiving=Get SkydivingState()) -> (bIsGliderRight=~, bIsGliderCenter=~, bIsGliderLeft=~, bIsGliderForward=~, bIsGliderBack=~, bHasDeployedAndIsStandingUpright=~).bIsGliderForward)]
 }
 AnimationTransitionGraph_1 {
  path=AnimGraph.AnimGraphNode_StateMachine_8.LeanTransitions.AnimStateNode_14.Lean_Out.AnimGraphNode_StateMachine_0.Lean_Out.AnimStateTransitionNode_26.AnimationTransitionGraph_1
  nodes [Result(bCanEnterTransition=Break Fort Anim Input Skydiving(FortAnimInput_Skydiving=Get SkydivingState()) -> (bIsGliderRight=~, bIsGliderCenter=~, bIsGliderLeft=~, bIsGliderForward=~, bIsGliderBack=~, bHasDeployedAndIsStandingUpright=~).bIsGliderBack)]
 }
 AnimationTransitionGraph_1 {
  path=AnimGraph.AnimGraphNode_StateMachine_8.LeanTransitions.AnimStateNode_14.Lean_Out.AnimGraphNode_StateMachine_0.Lean_Out.AnimStateTransitionNode_27.AnimationTransitionGraph_1
  nodes [Result(bCanEnterTransition=Break Fort Anim Input Skydiving(FortAnimInput_Skydiving=Get SkydivingState()) -> (bIsGliderRight=~, bIsGliderCenter=~, bIsGliderLeft=~, bIsGliderForward=~, bIsGliderBack=~, bHasDeployedAndIsStandingUpright=~).bIsGliderForward)]
 }
 AnimationTransitionGraph_1 {
  path=AnimGraph.AnimGraphNode_StateMachine_8.LeanTransitions.AnimStateNode_14.Lean_Out.AnimGraphNode_StateMachine_0.Lean_Out.AnimStateTransitionNode_28.AnimationTransitionGraph_1
  nodes [Result(bCanEnterTransition=Break Fort Anim Input Skydiving(FortAnimInput_Skydiving=Get SkydivingState()) -> (bIsGliderRight=~, bIsGliderCenter=~, bIsGliderLeft=~, bIsGliderForward=~, bIsGliderBack=~, bHasDeployedAndIsStandingUpright=~).bIsGliderBack)]
 }
 AnimationTransitionGraph_1 {
  path=AnimGraph.AnimGraphNode_StateMachine_8.LeanTransitions.AnimStateNode_14.Lean_Out.AnimGraphNode_StateMachine_0.Lean_Out.AnimStateTransitionNode_29.AnimationTransitionGraph_1
  nodes [Result(bCanEnterTransition=Break Fort Anim Input Skydiving(FortAnimInput_Skydiving=Get SkydivingState()) -> (bIsGliderRight=~, bIsGliderCenter=~, bIsGliderLeft=~, bIsGliderForward=~, bIsGliderBack=~, bHasDeployedAndIsStandingUpright=~).bIsGliderLeft)]
 }
 AnimationTransitionGraph_1 {
  path=AnimGraph.AnimGraphNode_StateMachine_8.LeanTransitions.AnimStateNode_14.Lean_Out.AnimGraphNode_StateMachine_0.Lean_Out.AnimStateTransitionNode_30.AnimationTransitionGraph_1
  nodes [
   n1 = Break Fort Anim Input Skydiving(FortAnimInput_Skydiving=Get SkydivingState()) -> (bIsGliderRight=~, bIsGliderCenter=~, bIsGliderLeft=~, bIsGliderForward=~, bIsGliderBack=~, bHasDeployedAndIsStandingUpright=~)
   Result(bCanEnterTransition=@n1)
   "Equal (Enum)"(A=@n1.LocalAccelDir, B=Right)
  ]
 }
 AnimationTransitionGraph_1 {
  path=AnimGraph.AnimGraphNode_StateMachine_8.LeanTransitions.AnimStateNode_14.Lean_Out.AnimGraphNode_StateMachine_0.Lean_Out.AnimStateTransitionNode_31.AnimationTransitionGraph_1
  nodes [Result(bCanEnterTransition=Break Fort Anim Input Skydiving(FortAnimInput_Skydiving=Get SkydivingState()) -> (bIsGliderRight=~, bIsGliderCenter=~, bIsGliderLeft=~, bIsGliderForward=~, bIsGliderBack=~, bHasDeployedAndIsStandingUpright=~).bIsGliderBack)]
 }
 AnimationTransitionGraph_1 {
  path=AnimGraph.AnimGraphNode_StateMachine_8.LeanTransitions.AnimStateNode_14.Lean_Out.AnimGraphNode_StateMachine_0.Lean_Out.AnimStateTransitionNode_32.AnimationTransitionGraph_1
  nodes [Result(bCanEnterTransition=Break Fort Anim Input Skydiving(FortAnimInput_Skydiving=Get SkydivingState()) -> (bIsGliderRight=~, bIsGliderCenter=~, bIsGliderLeft=~, bIsGliderForward=~, bIsGliderBack=~, bHasDeployedAndIsStandingUpright=~).bIsGliderForward)]
 }
 AnimationTransitionGraph_1 {
  path=AnimGraph.AnimGraphNode_StateMachine_8.LeanTransitions.AnimStateNode_14.Lean_Out.AnimGraphNode_StateMachine_0.Lean_Out.AnimStateTransitionNode_33.AnimationTransitionGraph_1
  nodes [Result(bCanEnterTransition=Break Fort Anim Input Skydiving(FortAnimInput_Skydiving=Get SkydivingState()) -> (bIsGliderRight=~, bIsGliderCenter=~, bIsGliderLeft=~, bIsGliderForward=~, bIsGliderBack=~, bHasDeployedAndIsStandingUpright=~).bIsGliderLeft)]
 }
 AnimationTransitionGraph_1 {
  path=AnimGraph.AnimGraphNode_StateMachine_8.LeanTransitions.AnimStateNode_14.Lean_Out.AnimGraphNode_StateMachine_0.Lean_Out.AnimStateTransitionNode_34.AnimationTransitionGraph_1
  nodes [
   n1 = Break Fort Anim Input Skydiving(FortAnimInput_Skydiving=Get SkydivingState()) -> (bIsGliderRight=~, bIsGliderCenter=~, bIsGliderLeft=~, bIsGliderForward=~, bIsGliderBack=~, bHasDeployedAndIsStandingUpright=~)
   Result(bCanEnterTransition=@n1)
   "Equal (Enum)"(A=@n1.LocalAccelDir, B=Right)
  ]
 }
 Lean_Into {
  path=AnimGraph.AnimGraphNode_StateMachine_8.LeanTransitions.AnimStateNode_17.Lean_Into
  nodes [
   AnimGraphNode_StateMachine_0 = Lean_Into() {is=title AnimGraphNode_StateMachine}
   Output Animation Pose(Result=@AnimGraphNode_StateMachine_0)
  ]
 }
 Lean_Into {
  entry=Lean_Zero
  path=AnimGraph.AnimGraphNode_StateMachine_8.LeanTransitions.AnimStateNode_17.Lean_Into.AnimGraphNode_StateMachine_0.Lean_Into
  ids=id
  nodes [
   Left Into() {is=title AnimStateAliasNode, also {aliases [Back_Into,Forward_Into,Lean_Zero,Right_Into]}}
   Forward Into() {is=title AnimStateAliasNode, also {aliases [Back_Into,Lean_Zero,Left_Into,Right_Into]}}
   Right Into() {is=title AnimStateAliasNode, also {aliases [Back_Into,Forward_Into,Lean_Zero,Left_Into]}}
   Back Into() {is=title AnimStateAliasNode, also {aliases [Forward_Into,Lean_Zero,Left_Into,Right_Into]}}
   AnimStateNode_0 = Lean_Zero()
   AnimStateNode_3 = Left_Into() {is=title AnimStateNode}
   AnimStateNode_4 = Right_Into() {is=title AnimStateNode}
   AnimStateNode_5 = Forward_Into() {is=title AnimStateNode}
   AnimStateNode_6 = Back_Into() {is=title AnimStateNode}
   AnimStateTransitionNode_1 = AnimStateTransitionNode()
   AnimStateTransitionNode_2 = AnimStateTransitionNode()
   AnimStateTransitionNode_3 = AnimStateTransitionNode()
   AnimStateTransitionNode_4 = AnimStateTransitionNode()
  ]
  links [
   {
    dst=Back_Into
    rule=AnimGraph.AnimGraphNode_StateMachine_8.LeanTransitions.AnimStateNode_17.Lean_Into.AnimGraphNode_StateMachine_0.Lean_Into.AnimStateTransitionNode_1.Transition
    src=Back Into
    via=AnimStateTransitionNode_1
   }
   {
    dst=Forward_Into
    rule=AnimGraph.AnimGraphNode_StateMachine_8.LeanTransitions.AnimStateNode_17.Lean_Into.AnimGraphNode_StateMachine_0.Lean_Into.AnimStateTransitionNode_3.Transition
    src=Forward Into
    via=AnimStateTransitionNode_3
   }
   {
    dst=Left_Into
    rule=AnimGraph.AnimGraphNode_StateMachine_8.LeanTransitions.AnimStateNode_17.Lean_Into.AnimGraphNode_StateMachine_0.Lean_Into.AnimStateTransitionNode_4.Transition
    src=Left Into
    via=AnimStateTransitionNode_4
   }
   {
    dst=Right_Into
    rule=AnimGraph.AnimGraphNode_StateMachine_8.LeanTransitions.AnimStateNode_17.Lean_Into.AnimGraphNode_StateMachine_0.Lean_Into.AnimStateTransitionNode_2.Transition
    src=Right Into
    via=AnimStateTransitionNode_2
   }
  ]
 }
 Lean_Zero {
  path=AnimGraph.AnimGraphNode_StateMachine_8.LeanTransitions.AnimStateNode_17.Lean_Into.AnimGraphNode_StateMachine_0.Lean_Into.AnimStateNode_0.Lean_Zero
  nodes [
   n1 = "Sequence Player (additive) 'CoconutShell_Lean_Center_Pose_GLIDER'"() {bLoopAnimation=false, sequence=&CoconutShell_Lean_Center_Pose_GLIDER}
   Output Animation Pose(Result=@n1)
  ]
 }
 Left_Into {
  path=AnimGraph.AnimGraphNode_StateMachine_8.LeanTransitions.AnimStateNode_17.Lean_Into.AnimGraphNode_StateMachine_0.Lean_Into.AnimStateNode_3.Left_Into
  nodes [
   n1 = "Sequence Player (additive) 'CoconutShell_PositionAdjustement_Pose_GLIDER'"() {
    bLoopAnimation=false
    sequence=&CoconutShell_PositionAdjustement_Pose_GLIDER
   }
   n2 = "Sequence Player (additive) 'CoconutShell_Lean_Left_Into_GLIDER'"() {bLoopAnimation=false, sequence=&/CoconutShell_Lean_Left_Into_GLIDER}
   Output Animation Pose(Result=Apply Additive!(Base=@n2, Additive=@n1, Alpha:float) -> (Pose) {alphaBoolBlend {blendOption=Linear}, alphaScaleBiasClamp {scale=0.35}})
  ]
 }
 Right_Into {
  path=AnimGraph.AnimGraphNode_StateMachine_8.LeanTransitions.AnimStateNode_17.Lean_Into.AnimGraphNode_StateMachine_0.Lean_Into.AnimStateNode_4.Right_Into
  nodes [
   n1 = "Sequence Player (additive) 'CoconutShell_PositionAdjustement_Pose_GLIDER'"() {
    bLoopAnimation=false
    sequence=&CoconutShell_PositionAdjustement_Pose_GLIDER
   }
   n2 = "Sequence Player (additive) 'CoconutShell_Lean_Right_Into_GLIDER'"() {bLoopAnimation=false, sequence=&/CoconutShell_Lean_Right_Into_GLIDER}
   Output Animation Pose(Result=Apply Additive(Base=@n2, Additive=@n1) {alphaBoolBlend {blendOption=Linear}, alphaScaleBiasClamp {scale=0.65}})
  ]
 }
 Forward_Into {
  path=AnimGraph.AnimGraphNode_StateMachine_8.LeanTransitions.AnimStateNode_17.Lean_Into.AnimGraphNode_StateMachine_0.Lean_Into.AnimStateNode_5.Forward_Into
  nodes [
   n1 = "Sequence Player (additive) 'CoconutShell_Lean_Forward_Into_GLIDER'"() {bLoopAnimation=false, sequence=&/CoconutShell_Lean_Forward_Into_GLIDER}
   n2 = "Sequence Player (additive) 'CoconutShell_PositionAdjustement_Pose_GLIDER'"() {
    bLoopAnimation=false
    sequence=&CoconutShell_PositionAdjustement_Pose_GLIDER
   }
   Output Animation Pose(Result=Apply Additive(Base=@n1, Additive=@n2) {alphaBoolBlend {blendOption=Linear}, alphaScaleBiasClamp {scale=0.2}})
  ]
 }
 Back_Into {
  path=AnimGraph.AnimGraphNode_StateMachine_8.LeanTransitions.AnimStateNode_17.Lean_Into.AnimGraphNode_StateMachine_0.Lean_Into.AnimStateNode_6.Back_Into
  nodes [
   n1 = "Sequence Player (additive) 'CoconutShell_Lean_Back_Into_GLIDER'"() {bLoopAnimation=false, sequence=&/CoconutShell_Lean_Back_Into_GLIDER}
   Output Animation Pose(Result=@n1)
  ]
 }
 Transition {
  path=AnimGraph.AnimGraphNode_StateMachine_8.LeanTransitions.AnimStateNode_17.Lean_Into.AnimGraphNode_StateMachine_0.Lean_Into.AnimStateTransitionNode_1.Transition
  nodes [Result(bCanEnterTransition=Break Fort Anim Input Skydiving(FortAnimInput_Skydiving=Get SkydivingState()) -> (bIsGliderRight=~, bIsGliderCenter=~, bIsGliderLeft=~, bIsGliderForward=~, bIsGliderBack=~, bHasDeployedAndIsStandingUpright=~).bIsGliderBack)]
 }
 Transition {
  path=AnimGraph.AnimGraphNode_StateMachine_8.LeanTransitions.AnimStateNode_17.Lean_Into.AnimGraphNode_StateMachine_0.Lean_Into.AnimStateTransitionNode_2.Transition
  nodes [Result(bCanEnterTransition=Break Fort Anim Input Skydiving(FortAnimInput_Skydiving=Get SkydivingState()) -> (bIsGliderRight=~, bIsGliderCenter=~, bIsGliderLeft=~, bIsGliderForward=~, bIsGliderBack=~, bHasDeployedAndIsStandingUpright=~))]
 }
 Transition {
  path=AnimGraph.AnimGraphNode_StateMachine_8.LeanTransitions.AnimStateNode_17.Lean_Into.AnimGraphNode_StateMachine_0.Lean_Into.AnimStateTransitionNode_3.Transition
  nodes [Result(bCanEnterTransition=Break Fort Anim Input Skydiving(FortAnimInput_Skydiving=Get SkydivingState()) -> (bIsGliderRight=~, bIsGliderCenter=~, bIsGliderLeft=~, bIsGliderForward=~, bIsGliderBack=~, bHasDeployedAndIsStandingUpright=~).bIsGliderForward)]
 }
 Transition {
  path=AnimGraph.AnimGraphNode_StateMachine_8.LeanTransitions.AnimStateNode_17.Lean_Into.AnimGraphNode_StateMachine_0.Lean_Into.AnimStateTransitionNode_4.Transition
  nodes [Result(bCanEnterTransition=Break Fort Anim Input Skydiving(FortAnimInput_Skydiving=Get SkydivingState()) -> (bIsGliderRight=~, bIsGliderCenter=~, bIsGliderLeft=~, bIsGliderForward=~, bIsGliderBack=~, bHasDeployedAndIsStandingUpright=~).bIsGliderLeft)]
 }
 AnimationTransitionGraph_0 {path=AnimGraph.AnimGraphNode_StateMachine_8.LeanTransitions.AnimStateTransitionNode_26.AnimationTransitionGraph_0, nodes [Result()]}
 AnimationTransitionGraph_0 {
  path=AnimGraph.AnimGraphNode_StateMachine_8.LeanTransitions.AnimStateTransitionNode_37.AnimationTransitionGraph_0
  nodes [
   n1 = Get bIsAccelerating2D()
   Result(bCanEnterTransition=@n1)
   n2 = Get Skydiving() -> (Skydiving:Fort Anim Input Skydiving Structure)
   OR Boolean!(
    A="Equal (Enum)"(
     A=Break Fort Anim Input Skydiving(FortAnimInput_Skydiving=@n2) -> (bIsGliderRight=~, bIsGliderCenter=~, bIsGliderLeft=~, bIsGliderForward=~, bIsGliderBack=~, bHasDeployedAndIsStandingUpright=~).LocalAccelDir
     B=Center
    )
    B=NOT Boolean(A=@n1)
   ) -> (ReturnValue)
  ]
 }
 AnimationTransitionGraph_0 {
  path=AnimGraph.AnimGraphNode_StateMachine_8.LeanTransitions.AnimStateTransitionNode_38.AnimationTransitionGraph_0
  nodes [
   n1 = NOT Boolean(A=Get bIsAccelerating2D())
   Result(bCanEnterTransition=@n1)
   n2 = Break Fort Anim Input Skydiving(FortAnimInput_Skydiving=Get Skydiving()) -> (bIsGliderRight=~, bIsGliderCenter=~, bIsGliderLeft=~, bIsGliderForward=~, bIsGliderBack=~, bHasDeployedAndIsStandingUpright=~)
   OR Boolean(A=@n2.bIsGliderCenter, B=@n1)
   "Equal (Enum)"(A=@n2.LocalAccelDir, B=Center)
  ]
 }
 AnimationTransitionGraph_0 {
  path=AnimGraph.AnimGraphNode_StateMachine_8.LeanTransitions.AnimStateTransitionNode_40.AnimationTransitionGraph_0
  nodes [Result(bCanEnterTransition=Get bIsAccelerating2D())]
 }
 DoubleFlapOnly {nodes [
   Inputs()
   > Do Once()
   > Flap 2x()
   Outputs()
  ]}
 EventGraph {nodes [
   HeadDartEvent() {is=title K2Node_CustomEvent}
   > n1 = Sequence()
   n2 = Break Fort Anim Input Skydiving!(FortAnimInput_Skydiving=Get SkydivingState()) -> (
    ParaGilderRootOffset:Vector=~
    ParaGliderRootRotationOverride:Rotator=~
    bUseParaGlideRootModifier:bool=~
    bIsSkydivingFromLaunchPad:bool=~
    bIsSkydivingFromBus:bool=~
    bIsInVortex:bool=~
    bIsDiving:bool=~
    bIsParachuteOpen:bool=~
    bIsSkydiving:bool=~
    bIsParachuteLeaning:bool=~
    bIsSkydiveLeaning:bool=~
    bIsLeaning:bool=~
    bIsSkydiveDiveMode:bool=~
    bPlaySkydiveDrift:bool=~
    bSkydiveDriftDelayActive:bool=~
    bSkydiveDriftAnimAllowed:bool=~
    bIsGliderRight=~
    bIsGliderCenter=~
    bIsGliderLeft=~
    bIsGliderForward=~
    bIsGliderBack=~
    bHasDeployedAndIsStandingUpright=~
    LocalAccelForward:float=~
    LocalAccelRight:float=~
    LocalVelocityRight:float=~
    SkydiveAimPitch:float=~
    SkydiveAimYaw:float=~
    DeployChuteAnimRate:float=~
    SkydiveDriftAnimRate:float=~
    SkydiveDriftAnimRateCurrent:float=~
    SkydiveFidgetAnimRate:float=~
    SkydiveFidgetAnimRateCurrent:float=~
    SkydiveAdditiveAlpha:float=~
    SkydiveDriftDelay:float=~
    SkydiveDriftAnim:int
    SkydiveDriftAnimMax:int
    LaunchpadAnim:int
    LocalAccelDir
    DirectionLast:ESkydivingDirection Enum
    PoseAdditive
   )
   > n3 = Branch(Condition=OR Boolean!(A=@n2.bIsGliderRight, B=@n2.bIsGliderLeft, C:bool="Not Equal (Float)"(A=@n2.SkydiveAimYaw)) -> (ReturnValue))
   n4 = Branch(execute=@n3.else, Condition=Get EnableHeadDarts() -> (EnableHeadDarts:bool))
   Delay(execute=[@n4.then, @n6.then], Duration=RandomFloatInRange(Min=10, Max=25))
   > n5 = Branch(Condition=RandomBool!() -> (ReturnValue:bool))
   > Set EnableHeadDarts(EnableHeadDarts:bool=false) -> (Output_Get:bool)
   > Set RightHeadDart(RightHeadDart:bool=true) -> (Output_Get:bool)
   Set EnableHeadDarts(execute=@n5.else, EnableHeadDarts=false)
   > Set LeftHeadDart(LeftHeadDart:bool=true) -> (Output_Get:bool)
   n6 = Branch(execute=@n1.then_1, Condition=Get FrontEnd())
   n7 = Event BlueprintUpdateAnimation()
   > Set Time(Time:double=@n7.DeltaTimeX) -> (Output_Get:double)
   > n8 = Sequence!(execute) -> (then_0, then_1, then_2:Exec, then_3:Exec)
   n9 = Sequence!(execute=@n8.then_1) -> (then_0, then_1, then_2, then_3, then_4:Exec, then_5:Exec, then_6:Exec, then_7:Exec)
   > n10 = Branch(Condition=Get FrontEnd())
   n11 = Break Fort Anim Input Skydiving(FortAnimInput_Skydiving=Get SkydivingState()) -> (
    ParaGilderRootOffset=~
    ParaGliderRootRotationOverride=~
    bUseParaGlideRootModifier=~
    bIsSkydivingFromLaunchPad=~
    bIsSkydivingFromBus=~
    bIsInVortex=~
    bIsDiving=~
    bIsParachuteOpen=~
    bIsSkydiving=~
    bIsParachuteLeaning=~
    bIsSkydiveLeaning=~
    bIsLeaning=~
    bIsSkydiveDiveMode=~
    bPlaySkydiveDrift=~
    bSkydiveDriftDelayActive=~
    bSkydiveDriftAnimAllowed=~
    bIsGliderRight=~
    bIsGliderCenter=~
    bIsGliderLeft=~
    bIsGliderForward=~
    bIsGliderBack=~
    bHasDeployedAndIsStandingUpright=~
    LocalAccelForward=~
    LocalAccelRight=~
    LocalVelocityRight=~
    SkydiveAimPitch=~
    SkydiveAimYaw=~
    DeployChuteAnimRate=~
    SkydiveDriftAnimRate=~
    SkydiveDriftAnimRateCurrent=~
    SkydiveFidgetAnimRate=~
    SkydiveFidgetAnimRateCurrent=~
    SkydiveAdditiveAlpha=~
    SkydiveDriftDelay=~
   )
   n12 = Branch(execute=@n10.else, Condition=OR Boolean!(A=@n11.bIsGliderRight, B=@n11.bIsGliderLeft) -> (ReturnValue))
   > n13 = Branch(Condition="Equal (Float)"(B=@n11.SkydiveAimYaw) {is=title K2Node_PromotableOperator})
   > Do Once(Reset=@n16.then_0)
   > Branch(Condition=Get EnableFlap() -> (EnableFlap:bool))
   > n14 = Branch(Condition=@n11.bIsGliderForward)
   Branch(execute=@n14.else, Condition=RandomBool())
   > n15 = Branch(Condition=RandomBool())
   > Set OneFlap(OneFlap:bool=true) -> (Output_Get:bool)
   Set TwoFlaps(execute=@n15.else, TwoFlaps:bool=true) -> (Output_Get:bool)
   n16 = Sequence!(execute=[@n13.else, @n12.else]) -> (then_0, then_1)
   Branch(execute=[@n34.then_1, @n16.then_1, @n40.then_1, @n37.then_1], Condition=Get EnableFlap())
   > n17 = Branch(Condition=Get FrontEnd())
   Delay(execute=@n17.else, Duration=RandomFloatInRange(Min=4, Max=8))
   n18 = Break Fort Anim Input Skydiving(FortAnimInput_Skydiving=Get SkydivingState()) -> (
    ParaGilderRootOffset=~
    ParaGliderRootRotationOverride=~
    bUseParaGlideRootModifier=~
    bIsSkydivingFromLaunchPad=~
    bIsSkydivingFromBus=~
    bIsInVortex=~
    bIsDiving=~
    bIsParachuteOpen=~
    bIsSkydiving=~
    bIsParachuteLeaning=~
    bIsSkydiveLeaning=~
    bIsLeaning=~
    bIsSkydiveDiveMode=~
    bPlaySkydiveDrift=~
    bSkydiveDriftDelayActive=~
    bSkydiveDriftAnimAllowed=~
    bIsGliderRight=~
    bIsGliderCenter=~
    bIsGliderLeft=~
    bIsGliderForward=~
    bIsGliderBack=~
    bHasDeployedAndIsStandingUpright=~
    LocalAccelForward=~
    LocalAccelRight=~
    LocalVelocityRight=~
    SkydiveAimPitch=~
    SkydiveAimYaw=~
    DeployChuteAnimRate=~
    SkydiveDriftAnimRate=~
    SkydiveDriftAnimRateCurrent=~
    SkydiveFidgetAnimRate=~
    SkydiveFidgetAnimRateCurrent=~
    SkydiveAdditiveAlpha=~
    SkydiveDriftDelay=~
   )
   n19 = OR Boolean!(
    A=@n18.bIsGliderRight
    B=@n18.bIsGliderLeft
    C:bool=@n18.bIsGliderForward
    D:bool=@n18.bIsGliderBack
    E:bool="Not Equal (Float)"(B=@n18.SkydiveAimYaw)
   ) -> (ReturnValue)
   > Branch(Condition=AND Boolean(A=Get EnableFlap(), B=NOT Boolean(A=@n19)))
   > Branch(Condition=RandomBool())
   > n20 = Branch(Condition=RandomBool())
   > Set OneFlap(OneFlap=true)
   Set TwoFlaps(execute=@n20.else, TwoFlaps=true)
   n21 = Get TwoFlaps()
   n22 = Get OneFlap()
   Branch(
    execute=@n9.then_3
    Condition=AND Boolean!(A=NOT Boolean(A=@n21), B=NOT Boolean(A=@n22), C:bool=Get IntoEnd() -> (IntoEnd:bool)) -> (ReturnValue)
   )
   > Set EnableFlap(EnableFlap:bool=true) -> (Output_Get:bool)
   n23 = Branch(execute=@n9.then_6, Condition=AND Boolean!(A=Get FirstBlink(), B=Get Caw()) -> (ReturnValue))
   Delay(execute=@n23.else, Duration=RandomFloatInRange(Min=4, Max=10))
   > Set FirstBlink(FirstBlink:bool=true) -> (Output_Get:bool)
   Branch(execute=@n9.then_7, Condition=Get EnableCaw() -> (EnableCaw:bool))
   > Delay(Duration=RandomFloatInRange(Min=20, Max=75))
   > n24 = Sequence()
   n25 = OR Boolean!(A=Get LeftHeadDart(), B=Get RightHeadDart()) -> (ReturnValue)
   > n26 = Branch(Condition=AND Boolean(A=NOT Boolean(A=@n25), B=Get EnableCaw()))
   Set EnableCaw(execute=[@n27.then, @n26.then], EnableCaw:bool=false) -> (Output_Get:bool)
   > Set Caw(Caw:bool=true) -> (Output_Get:bool)
   n27 = Branch(execute=@n24.then_1, Condition=AND Boolean(A=Get FrontEnd(), B=Get IntoEnd()))
   n28 = Branch(execute=@n9.then_5, Condition=Get FrontEnd())
   n29 = Break Fort Anim Input Skydiving(FortAnimInput_Skydiving=Get SkydivingState()) -> (
    ParaGilderRootOffset=~
    ParaGliderRootRotationOverride=~
    bUseParaGlideRootModifier=~
    bIsSkydivingFromLaunchPad=~
    bIsSkydivingFromBus=~
    bIsInVortex=~
    bIsDiving=~
    bIsParachuteOpen=~
    bIsSkydiving=~
    bIsParachuteLeaning=~
    bIsSkydiveLeaning=~
    bIsLeaning=~
    bIsSkydiveDiveMode=~
    bPlaySkydiveDrift=~
    bSkydiveDriftDelayActive=~
    bSkydiveDriftAnimAllowed=~
    bIsGliderRight=~
    bIsGliderCenter=~
    bIsGliderLeft=~
    bIsGliderForward=~
    bIsGliderBack=~
    bHasDeployedAndIsStandingUpright=~
    LocalAccelForward=~
    LocalAccelRight=~
    LocalVelocityRight=~
    SkydiveAimPitch=~
    SkydiveAimYaw=~
    DeployChuteAnimRate=~
    SkydiveDriftAnimRate=~
    SkydiveDriftAnimRateCurrent=~
    SkydiveFidgetAnimRate=~
    SkydiveFidgetAnimRateCurrent=~
    SkydiveAdditiveAlpha=~
    SkydiveDriftDelay=~
   )
   n30 = Branch(execute=@n28.else, Condition="Not Equal (Float)"(B=@n29.SkydiveAimYaw))
   > Do Once(Reset=@n34.then_0)
   > Branch(Condition=Get EnableFlap())
   > n31 = Branch(Condition=@n29.bIsGliderForward)
   > Branch(Condition=RandomBool())
   > Set FlapForward(FlapForward:bool=true) -> (Output_Get:bool)
   n32 = Branch(execute=@n31.else, Condition=@n29.bIsGliderBack)
   > Branch(Condition=RandomBool())
   > Set FlapBack(FlapBack:bool=true) -> (Output_Get:bool)
   Branch(execute=@n32.else, Condition=RandomBool())
   > n33 = Branch(Condition=RandomBool())
   > Set OneFlap(OneFlap=true)
   Set TwoFlaps(execute=@n33.else, TwoFlaps=true)
   n34 = Sequence(execute=@n30.else)
   n35 = Branch(execute=@n9.then_4, Condition=Get FrontEnd())
   n36 = Branch(
    execute=@n35.else
    Condition=Break Fort Anim Input Skydiving(FortAnimInput_Skydiving=Get SkydivingState()) -> (
     ParaGilderRootOffset=~
     ParaGliderRootRotationOverride=~
     bUseParaGlideRootModifier=~
     bIsSkydivingFromLaunchPad=~
     bIsSkydivingFromBus=~
     bIsInVortex=~
     bIsDiving=~
     bIsParachuteOpen=~
     bIsSkydiving=~
     bIsParachuteLeaning=~
     bIsSkydiveLeaning=~
     bIsLeaning=~
     bIsSkydiveDiveMode=~
     bPlaySkydiveDrift=~
     bSkydiveDriftDelayActive=~
     bSkydiveDriftAnimAllowed=~
     bIsGliderRight=~
     bIsGliderCenter=~
     bIsGliderLeft=~
     bIsGliderForward=~
     bIsGliderBack=~
     bHasDeployedAndIsStandingUpright=~
     LocalAccelForward=~
     LocalAccelRight=~
     LocalVelocityRight=~
     SkydiveAimPitch=~
     SkydiveAimYaw=~
     DeployChuteAnimRate=~
     SkydiveDriftAnimRate=~
     SkydiveDriftAnimRateCurrent=~
     SkydiveFidgetAnimRate=~
     SkydiveFidgetAnimRateCurrent=~
     SkydiveAdditiveAlpha=~
     SkydiveDriftDelay=~
    ).bIsGliderForward
   )
   > Do Once(Reset=@n37.then_0)
   > Branch(Condition=Get EnableFlap())
   > Branch(Condition=RandomBool())
   n37 = Sequence(execute=@n36.else)
   n38 = Branch(execute=@n9.then_1, Condition=Get FrontEnd())
   n39 = Branch(
    execute=@n38.else
    Condition=Break Fort Anim Input Skydiving(FortAnimInput_Skydiving=Get SkydivingState()) -> (
     ParaGilderRootOffset=~
     ParaGliderRootRotationOverride=~
     bUseParaGlideRootModifier=~
     bIsSkydivingFromLaunchPad=~
     bIsSkydivingFromBus=~
     bIsInVortex=~
     bIsDiving=~
     bIsParachuteOpen=~
     bIsSkydiving=~
     bIsParachuteLeaning=~
     bIsSkydiveLeaning=~
     bIsLeaning=~
     bIsSkydiveDiveMode=~
     bPlaySkydiveDrift=~
     bSkydiveDriftDelayActive=~
     bSkydiveDriftAnimAllowed=~
     bIsGliderRight=~
     bIsGliderCenter=~
     bIsGliderLeft=~
     bIsGliderForward=~
     bIsGliderBack=~
     bHasDeployedAndIsStandingUpright=~
     LocalAccelForward=~
     LocalAccelRight=~
     LocalVelocityRight=~
     SkydiveAimPitch=~
     SkydiveAimYaw=~
     DeployChuteAnimRate=~
     SkydiveDriftAnimRate=~
     SkydiveDriftAnimRateCurrent=~
     SkydiveFidgetAnimRate=~
     SkydiveFidgetAnimRateCurrent=~
     SkydiveAdditiveAlpha=~
     SkydiveDriftDelay=~
    ).bIsGliderBack
   )
   > Do Once(Reset=@n40.then_0)
   > Branch(Condition=Get EnableFlap())
   > Branch(Condition=RandomBool())
   > Set FlapBack(FlapBack=true)
   n40 = Sequence(execute=@n39.else)
   Branch(execute=@n9.then_2, Condition=AND Boolean(A=Get FrontEnd(), B=Get IntoEnd()))
   > Set FrontEnd(FrontEnd:bool=true) -> (Output_Get:bool)
   > Delay(Duration=RandomFloatInRange(Min=4, Max=8))
   > Branch(Condition=AND Boolean(A=RandomBool(), B=Get FrontEnd()))
   > n41 = Branch(Condition=RandomBool())
   > Set OneFlap(OneFlap=true)
   Set TwoFlaps(execute=@n41.else, TwoFlaps=true)
   Branch(
    execute=@n8.then_2
    Condition=IsCosmeticObjectDisplayedInFrontendContext!(CosmeticObject:Object Reference=Self-Reference()) -> (ReturnValue:bool)
   )
   > Set FrontEnd(FrontEnd=true)
   Get Is Parachute Open!(execute:Exec=@n8.then_3) {is=title K2Node_MacroInstance}
   n42 = ForwardBack_Value_Interp!(
    execute=@n8.then_0
    Time:double=Get Time() -> (Time:double)
    Currnet_Value:double=Get Acceleration() -> (Acceleration:double)
    Idle_Value:double=1
    Idle_InterpSpeed:double=1
    Forward_Value:double=2
    Forward_InterpSpeed:double=2
    Back_Value:double=1
    Back_InterpSpeed:double=122
    LocalAccelForward:double=Break Fort Anim Input Skydiving(FortAnimInput_Skydiving=Get SkydivingState()) -> (
     ParaGilderRootOffset=~
     ParaGliderRootRotationOverride=~
     bUseParaGlideRootModifier=~
     bIsSkydivingFromLaunchPad=~
     bIsSkydivingFromBus=~
     bIsInVortex=~
     bIsDiving=~
     bIsParachuteOpen=~
     bIsSkydiving=~
     bIsParachuteLeaning=~
     bIsSkydiveLeaning=~
     bIsLeaning=~
     bIsSkydiveDiveMode=~
     bPlaySkydiveDrift=~
     bSkydiveDriftDelayActive=~
     bSkydiveDriftAnimAllowed=~
     bIsGliderRight=~
     bIsGliderCenter=~
     bIsGliderLeft=~
     bIsGliderForward=~
     bIsGliderBack=~
     bHasDeployedAndIsStandingUpright=~
     LocalAccelForward=~
     LocalAccelRight=~
     LocalVelocityRight=~
     SkydiveAimPitch=~
     SkydiveAimYaw=~
     DeployChuteAnimRate=~
     SkydiveDriftAnimRate=~
     SkydiveDriftAnimRateCurrent=~
     SkydiveFidgetAnimRate=~
     SkydiveFidgetAnimRateCurrent=~
     SkydiveAdditiveAlpha=~
     SkydiveDriftDelay=~
    ).LocalAccelForward
   ) -> (then, New_Value:double)
   > Set Acceleration(Acceleration:double=@n42) -> (Output_Get:double)
   AnimNotify_TurnONEngines() {is=title K2Node_Event}
   > SetGliderFullyDeployed!(execute, self:Fort Glider Anim Instance Object Reference, bIsFullyDeployed:bool=true) -> (then)
   Event BlueprintBeginPlay() {is=title K2Node_Event}
   > Set IntoEnd(IntoEnd:bool=false) -> (Output_Get:bool)
   AnimNotify_IntoStart() {is=title K2Node_Event}
   > Set IntoEnd(IntoEnd=false)
   n43 = AnimNotify_ClosedEntered() {is=title K2Node_Event}
   n44 = GetOwningComponent()
   HideBoneByName!(
    execute=[@n43.then, @n45.then]
    self:Skinned Mesh Component Object Reference=@n44
    BoneName:Name=Glider_Base
    PhysBodyOption:EPhysBodyOp Enum=PBO_None
   ) -> (then)
   AnimNotify_DeployEntered() {is=title K2Node_Event}
   > Delay(Duration=0.2)
   > UnHideBoneByName!(execute, self:Skinned Mesh Component Object Reference=@n44, BoneName:Name=Glider_Base) -> (then)
   n45 = Event BlueprintInitializeAnimation()
   AnimNotify_flapStart() {is=title K2Node_Event}
   > Set EnableFlap(EnableFlap=false)
   AnimNotify_HeadDartEnd() {is=title K2Node_Event}
   > Set EnableHeadDarts(EnableHeadDarts=true)
   > HeadDartEvent() {is=title K2Node_CallFunction}
   Event BlueprintPostEvaluateAnimation() {is=title K2Node_Event}
   > n46 = Sequence!(execute) -> (then_0, then_1, then_2:Exec, then_3:Exec)
   > Set Caw(Caw=false)
   > Set EnableCaw(EnableCaw=true)
   Set OneFlap(execute=@n46.then_1, OneFlap=false)
   > Set TwoFlaps(TwoFlaps=false)
   > Set FlapForward(FlapForward=false)
   > Set FlapBack(FlapBack=false)
   > Set EnableFlap(EnableFlap=true)
   Set RightHeadDart(execute=@n46.then_2, RightHeadDart=false)
   > Set LeftHeadDart(LeftHeadDart=false)
   > Set EnableHeadDarts(EnableHeadDarts=true)
   Set FirstBlink(execute=@n46.then_3, FirstBlink=false)
   AnimNotify_IntoEnd() {is=title K2Node_Event}
   > Set IntoEnd(IntoEnd=true)
   > Set EnableFlap(EnableFlap=true)
   > Set EnableHeadDarts(EnableHeadDarts=true)
   > Set EnableBlinks(EnableBlinks:bool=true) -> (Output_Get:bool)
   > HeadDartEvent()
   AnimNotify_flapEnd() {is=title K2Node_Event}
   > Set EnableFlap(EnableFlap=true)
   Branch(Condition=RandomBool())
   > Set FlapForward(FlapForward=true)
   IsAnimationPlaying!(Anim:Razer Chroma Asset Object Reference=Cast To RazerChromaAnimationAsset!(execute, Object:Wildcard) -> (then, CastFailed, AsRazer Chroma Asset:Razer Chroma Asset Object Reference) {is=title K2Node_DynamicCast}) -> (ReturnValue)
   Set FlapForward(FlapForward=true)
  ]}
 GetIsParachuteOpen {nodes [
   Inputs()
   > n1 = Branch(Condition=Break Fort Anim Input Skydiving(FortAnimInput_Skydiving=Get SkydivingState()) -> (
     ParaGilderRootOffset=~
     ParaGliderRootRotationOverride=~
     bUseParaGlideRootModifier=~
     bIsSkydivingFromLaunchPad=~
     bIsSkydivingFromBus=~
     bIsInVortex=~
     bIsDiving=~
     bIsParachuteOpen=~
     bIsSkydiving=~
     bIsParachuteLeaning=~
     bIsSkydiveLeaning=~
     bIsLeaning=~
     bIsSkydiveDiveMode=~
     bPlaySkydiveDrift=~
     bSkydiveDriftDelayActive=~
     bSkydiveDriftAnimAllowed=~
     bIsGliderRight=~
     bIsGliderCenter=~
     bIsGliderLeft=~
     bIsGliderForward=~
     bIsGliderBack=~
     bHasDeployedAndIsStandingUpright=~
     LocalAccelForward=~
     LocalAccelRight=~
     LocalVelocityRight=~
     SkydiveAimPitch=~
     SkydiveAimYaw=~
     DeployChuteAnimRate=~
     SkydiveDriftAnimRate=~
     SkydiveDriftAnimRateCurrent=~
     SkydiveFidgetAnimRate=~
     SkydiveFidgetAnimRateCurrent=~
     SkydiveAdditiveAlpha=~
     SkydiveDriftDelay=~
    ).bIsParachuteOpen)
   > Set IsParachuteOpen(IsParachuteOpen:bool=true) -> (Output_Get:bool)
   Set IsParachuteOpen(execute=@n1.else, IsParachuteOpen=false)
   Outputs()
  ]}
 GetPlayerOpenGliderBool {nodes [
   Inputs()
   > n1 = Branch(Condition=Get IsCosmeticPreview(self=Get OwnerGlider()))
   Outputs!(then:Exec=[@n3.then, @n1.then])
   n2 = Get OwnerFortPawn() -> (OwnerFortPawn:Fort Player Pawn Object Reference)
   n3 = Set bForceOpen(
    execute=@n1.else
    bForceOpen:bool=ShouldForceOpenGlider!(self:Fort Player Animation Interface Object Reference=GetAnimationInterface!(self:Fort Player Pawn Object Reference=@n2) -> (ReturnValue:Fort Player Animation Interface Object Reference)) -> (ReturnValue:bool)
   ) -> (Output_Get:bool)
  ]}
 RandomFlap {nodes [
   Inputs()
   Outputs!()
  ]}
 SimpleDoubleFlaps {nodes [
   Inputs()
   > Do Once()
   > n1 = Branch(Condition=RandomBool())
   > Flap 1x()
   Flap 2x(execute=@n1.else)
   Outputs()
  ]}
]
variables [
 Acceleration {type=number}
 Caw {type=boolean}
 CurrentFlapFloat {type=number}
 EnableBlinks {type=boolean}
 EnableCaw {type=boolean}
 EnableFlap {type=boolean, default=true}
 EnableHeadDarts {type=boolean}
 FirstBlink {type=boolean}
 FlapBack {type=boolean}
 FlapForward {type=boolean}
 FreezedFlapFloat {type=number}
 FrontEnd {type=boolean}
 IntoEnd {type=boolean}
 IsParachuteOpen {type=boolean}
 LeftHeadDart {type=boolean}
 OneFlap {type=boolean}
 RandomHeadDartsInt {type=integer}
 RightHeadDart {type=boolean}
 Time {type=number}
 TwoFlaps {type=boolean}
]
